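# Translucent unlit glow material for the SaxoPop pickaxe, built for Niagara mesh particles.
# Static switches select a scrolling, noise-distorted iridescent curve-atlas holofoil, a two-color height-gradient "galaxy" look, or a base color, feeding an emissive adjustment function.
# The base color is particle color brightened by an axis-aligned fresnel, passed through a named reroute.
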Material /BRCosmetics/Weapons/FORT_Melee/Pickaxe_SaxoPop/FX/Material/M_SaxopopGlow
graph {
 properties {
  bAutomaticallySetUsageInEditor=false
  bUsedWithNiagaraMeshParticles=true
  blendMode=BLEND_Translucent
  shadingModel=MSM_Unlit
  usedShadingModels=Unlit
 }
 nodes [
  n1 = TextureCoordinate()
  n2 = TextureSample() {samplerType=SAMPLERTYPE_Masks, texture=/Game/Characters/Player/Female/Medium/Bodies/F_MED_Neon_Tight_Suit/Materials/Soft_Perlin_output}
  n3 = Multiply(A=Subtract(A=@n2.R) {constB=0.5}, B=ScalarParameter() {defaultValue=0.5, parameterName=Distortion Strength, sliderMax=1}) {constB=0.1}
  n4 = Saturate(None=ComponentMask(None=Add(A=@n1, B=@n3)) {g=true})
  n5 = Time()
  n6 = CurveAtlasRowParameter(CurveTime=Frac(None=Add(A=@n4, B=@n5))) {
   atlas=/Game/Weapons/FORT_Melee/Pickaxe_RainbowStraps/FX/Holofoil/CA_Holofoil_Iridescent
   curve=/Game/Athena/LinearColorCurves/Curve_Holofoil_Iridescent
   parameterName=CurveAtlas
  }
  n7 = ScalarParameter() {defaultValue=-1, parameterName=Holo_Desasurate}
  n8 = WorldPosition()
  n9 = ComponentMask(None=ObjectBounds()) {b=true}
  n10 = Divide(
   A=ComponentMask(None=TransformPosition(None=@n8.XYZ) {transformSourceType=TRANSFORMPOSSOURCE_World}) {b=true}
   B=Multiply(A=Multiply(A=@n9) {constB=0.1}) {constB=1.2}
  )
  n11 = Desaturation(
   None=LinearInterpolate(
    A=VectorParameter() {parameterName=Color_01, defaultValue=<rgba 0 0.6488 0.8 1>}.RGB
    B=VectorParameter() {parameterName=Color_02, defaultValue=<rgba 1 0 0.9741 1>}.RGB
    Alpha=Clamp(None=@n10)
   )
   Fraction=ScalarParameter() {parameterName=Galaxy_Desaturate}
  )
  n12 = NamedRerouteUsage()
  MaterialFunctionCall(Emissive=StaticSwitchParameter(True=Desaturation(None=@n6, Fraction=@n7), False=StaticSwitchParameter(True=@n11, False=@n12) {parameterName=IsGalaxy}) {parameterName=IsHolofoil}) {materialFunction=&MF_FX_EmissiveAdjustments}
  n13 = ParticleColor()
  n14 = Transform(None=Constant3Vector() {constant=<rgba 0 1 0 0>}) {transformType=TRANSFORM_Local}
  n15 = Constant() {r=3}
  NamedRerouteDeclaration(None=LinearInterpolate(
    A=Multiply(A=@n13.RGB) {constB=0.5}
    B=@n13.RGB
    Alpha=MaterialFunctionCall(Axis=@n14, Exponent=@n15) {materialFunction=&AxisAlignedFresnel}.Fresnel
   )) {name=BaseParticleColor, nodeColor=<rgba 0.125 1 0 1>}
  Reroute(None=@n13.A)
 ]
}
