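# Idle Niagara effect for the SaxoPop pickaxe, adding a colored glow and twinkling sparkles.
# A burst-spawned Glow emitter drives mesh and sprite renderers, with color and alpha scaled by a system Timeline float.
# Rate-spawned Sparkles place sprites at skeletal mesh sockets with cylinder offsets, flickering through a randomized size curve and fading out.
# User.VariantPicker selects one of six color options in both emitters.
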NiagaraSystem /BRCosmetics/Weapons/FORT_Melee/Pickaxe_SaxoPop/FX/NS_Pickaxe_SaxoPop_Idle
dependencies {
 usedDataInterfaces [
  &NiagaraDataInterfaceMeshRendererInfo
  &NiagaraDataInterfaceCurve
  &NiagaraDataInterfaceSkeletalMesh
  &NiagaraDataInterfaceArrayFloat
 ]
 usedDynamicInputs [
  &SetBoolByIntComparison
  &Boolean_NotOperation
  &InvertBool
  &MakeCustomFloatFromBool
  &FNTimeOfDayScalar
 ]
 usedModules [
  &self:SystemSpawnScript.SystemScriptSource.SystemScriptGraph.NiagaraNodeAssignment_1.SetVariables_#1
  &self:SystemSpawnScript.SystemScriptSource.SystemScriptGraph.NiagaraNodeAssignment_3.SetVariables_#2
  &self:SystemSpawnScript.SystemScriptSource.SystemScriptGraph.NiagaraNodeAssignment_4.SetVariables_#3
  &SystemState
  &Timeline
  &self:SystemSpawnScript.SystemScriptSource.SystemScriptGraph.NiagaraNodeAssignment_5.SetVariables_#4
  &self:SystemSpawnScript.SystemScriptSource.SystemScriptGraph.NiagaraNodeAssignment_6.SetVariables_#5
  &self:SystemSpawnScript.SystemScriptSource.SystemScriptGraph.NiagaraNodeAssignment_0.SetVariables_#6
  &self:SystemSpawnScript.SystemScriptSource.SystemScriptGraph.NiagaraNodeAssignment_2.SetVariables_#7
 ]
}
emitters {
 Glow {
  emitterUpdateScript [
   &EmitterState(Life Cycle Mode=::NewEnumerator1, Loop Behavior=::NewEnumerator1, Loop Duration Mode=::NewEnumerator1, UseLoopDelay=false)
   &SpawnBurst_Instantaneous(Use Spawn Probability=false, UseLoopCountLimit=false)
  ]
  particleSpawnScript [
   &InitializeParticle(
    Color Mode=::NewEnumerator0
    Lifetime=0.25
    Sprite Size=<15 30>
    Sprite Size Mode=::NewEnumerator4
    UsePositionOffset=false
    Write Material Random=true
   )
   &InitialMeshOrientation(Mesh Axis To Orient=<0 -1 0>, Mesh Orientation Mode=::NewEnumerator2, Orientation Vector=<0 0 1>, Use Rotation Vector=false)
   &SpriteFacingAndAlignment(Sprite Alignment=<0 0 1>, WriteAlignment=true, WriteFacing=true)
   /SetVariables(Particles.Glowpos:NiagaraPosition=&AddVectorToPosition(Position=$Engine.Emitter.SimulationPosition, Vector=<0 0 27>)) {isSetParameters=true}
  ]
  particleUpdateScript [
   &ParticleState(Kill Particles When Lifetime Has Elapsed=false, Let Infinitely Lived Particles Die When Emitter Deactivates=true)
   &ScaleColor(
    Scale Alpha=$System.Timeline.Float
    Scale RGB=&MakeVectorFromLinearColorRGB(Color=&MergeHelper_ColorSwitch(
      0=&Multiply_LinearColorByFloat(Linear Color=<rgba 1 0.1094 0.1589 1>, Float=&Lerp_Float(A=0, B=5, Alpha=$System.Timeline.Float))
      1=&Multiply_LinearColorByFloat(Linear Color=<rgba 0.9387 0.2307 0.7682 1>, Float=&Lerp_Float(A=0, B=2, Alpha=$System.Timeline.Float))
      2=&Multiply_LinearColorByFloat(Linear Color=<rgba 0.9473 0.4518 0.1163 1>, Float=&Lerp_Float(A=0, B=4, Alpha=$System.Timeline.Float))
      3=&Multiply_LinearColorByFloat(Linear Color=<rgba 0.1119 0.5457 0.7305 1>, Float=&Lerp_Float(A=0, B=4, Alpha=$System.Timeline.Float))
      4=&Multiply_LinearColorByFloat(Linear Color=<rgba 0 0.6613 0.7605 1>, Float=&Lerp_Float(A=0, B=4, Alpha=$System.Timeline.Float))
      5=&Multiply_LinearColorByFloat(Linear Color=<rgba 1 1 1 1>, Float=&Lerp_Float(A=0, B=1, Alpha=$System.Timeline.Float))
      Index=&MakeIntFromFloat(Float=$User.VariantPicker, Conversion Mode=::NewEnumerator1)
      Num=::NewEnumerator4
     ))
    ScaleA=true
    ScaleRGB=true
   )
   /SetVariables(Particles.GlowColor=&Multiply_LinearColorByFloat(Linear Color=$Particles.Color, Float=0.3)) {isSetParameters=true}
  ]
  rendererClasses [&NiagaraMeshRendererProperties, &NiagaraSpriteRendererProperties]
  renderers [
   {rendererClass=&NiagaraMeshRendererProperties, rendererIndex=0}
   {rendererClass=&NiagaraSpriteRendererProperties, rendererIndex=1}
   {rendererClass=&NiagaraMeshRendererProperties, rendererIndex=2}
   {rendererClass=&NiagaraMeshRendererProperties, rendererIndex=3}
  ]
 }
 Sparkles {
  emitterUpdateScript [
   &EmitterState()
   &SpawnRate(Use Spawn Probability=false)
  ]
  particleSpawnScript [
   &InitializeParticle(
    Color=&MergeHelper_ColorSwitch(
     0=&RandomRangeLinearColor(Minimum=<rgba 0.9936 1 0.2601 1>, Maximum=<rgba 3 1.2731 0.8176 1>)
     1=&RandomRangeLinearColor(Minimum=<rgba 0.9974 1 0.7031 1>, Maximum=<rgba 3 0.8176 2.4117 1>)
     2=&RandomRangeLinearColor(Minimum=<rgba 0.9978 1 0.75 1>, Maximum=<rgba 2 0.4262 0.0729 1>)
     3=&RandomRangeLinearColor(Minimum=<rgba 0.9972 1 0.6771 1>, Maximum=<rgba 1.2604 2 1.9035 1>)
     4=&RandomRangeLinearColor(Minimum=<rgba 0.6927 1 0.9759 1>, Maximum=<rgba 0 2 1.9478 1>)
     5=&RandomRangeLinearColor(Minimum=<rgba 1 0.5417 0.9901 1>, Maximum=<rgba 1.0521 2 1.9753 1>)
     Index=&MakeIntFromFloat(Float=$User.VariantPicker, Conversion Mode=::NewEnumerator1)
     Num=::NewEnumerator4
    )
    Lifetime Max=1
    Lifetime Min=0.3
    Lifetime Mode=::NewEnumerator1
    Sprite Size Mode=::NewEnumerator1
    Uniform Sprite Size Max=15
    Uniform Sprite Size Min=13
    UsePositionOffset=false
    Write Material Random=true
   )
   &SkeletalMeshLocation(MaskSpawnGroup=false, Mesh Sampling Type=::NewEnumerator5, Skeletal Mesh=$System.Skeletal Mesh, Socket Sampling Mode=::NewEnumerator5)
   &ApplyInitialForces()
   &ShapeLocation(
    Cylinder Height=20
    Cylinder Radius=4
    MaskSpawnGroup=false
    Offset Mode=::NewEnumerator0
    Override Seed=false
    Shape Primitive=::NewEnumerator1
    Surface Only=false
    UseOwnerScale=true
   )
   &CameraOffset(Camera Offset Amount=25)
  ]
  particleUpdateScript [
   &ParticleState()
   &ScaleSpriteSize(
    Uniform Curve Index=&Add_Float(
     A=&Multiply_Float(A=$Particles.Age, B=&RandomRangeFloat(Minimum=0.8, Maximum=1.5))
     B=&RandomRangeFloat(Minimum=-0.25, Maximum=0.25)
    )
    Uniform Curve Sprite Scale=DataInterface() {propertyValues {Curve {keys [
        {tangentMode=RCTM_User, time=0.2, value=0}
        <at 0.22 1 RCIM_Cubic>
        {tangentMode=RCTM_User, time=0.5, value=0}
        <at 0.68 0>
        <at 0.7 0.8 RCIM_Cubic>
        <at 0.85 0>
       ]}}}
   )
   &Drag(Drag=2, Use Linear Drag=true, Use Rotational Drag=false)
   &SolveForcesAndVelocity(Clamp Velocity=false, Limit Acceleration=false)
   &ScaleColor(
    Scale Alpha=&FloatFromCurve(FloatCurve=DataInterface() {propertyValues {Curve {keys [<at 0.5 1 RCIM_Cubic>,<at 1 0>]}}})
    Scale RGB=&VectorFromFloat(Value=&FloatFromCurve(FloatCurve=DataInterface() {propertyValues {Curve {keys [{arriveTangent=-13.9762, interpMode=RCIM_Cubic, leaveTangent=-13.9762, tangentMode=RCTM_User, time=0, value=5}, <at 0.5 1>]}}}))
    ScaleA=true
    ScaleRGB=true
   )
  ]
  rendererClasses [&NiagaraSpriteRendererProperties]
  renderers [{rendererClass=&NiagaraSpriteRendererProperties}]
 }
}
summary {
 systemName=NS_Pickaxe_SaxoPop_Idle
 emitters [
  {emitterName=Sparkles, rendererClasses [&NiagaraSpriteRendererProperties]}
  {emitterName=Glow, rendererClasses [&NiagaraMeshRendererProperties, &NiagaraSpriteRendererProperties]}
 ]
 userVariables [
  User.bHasPickaxe {type=NiagaraBool, defaultValue=true}
  User.bIsFrontEndPreview {type=NiagaraBool}
  User.bIsOffhand {type=NiagaraBool}
  User.bisSwinging {type=NiagaraBool}
  User.VariantPicker {type=NiagaraFloat, defaultValue=0, description=""}
 ]
}
