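# Niagara impact effect for the SaxoPop pickaxe, a short burst of candy bits, layered flares, star and spike flashes with a light, galaxy smoke and curl-noise sparks.
# User.VariantPicker indexes color switches across most emitters to give six palette variants.
# Front-end modifiers adjust sizes and velocities for menu display, and time-of-day functions dim some layers.
# Custom particle attributes let several renderers share one particle with separate sizes, colors and lifetimes.
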NiagaraSystem /BRCosmetics/Weapons/FORT_Melee/Pickaxe_SaxoPop/FX/NS_Pickaxe_SaxoPop_Impact
dependencies {
 usedDataInterfaces [
  &NiagaraDataInterfaceCurve
  &NiagaraDataInterfacePlatformSet
  &NiagaraDataInterfaceColorCurve
  &NiagaraDataInterfaceVector2DCurve
 ]
 usedDynamicInputs [&MakeVector2D, &SetBoolByIntComparison]
 usedModules [
  &self:SystemSpawnScript.SystemScriptSource.SystemScriptGraph.NiagaraNodeAssignment_0.SetVariables_#1
  &self:SystemSpawnScript.SystemScriptSource.SystemScriptGraph.NiagaraNodeAssignment_1.SetVariables_#2
  &self:SystemSpawnScript.SystemScriptSource.SystemScriptGraph.NiagaraNodeAssignment_2.SetVariables_#3
  &SystemState
 ]
}
emitters {
 CandyBits {
  emitterUpdateScript [
   &EmitterState(Life Cycle Mode=::NewEnumerator1, Loop Behavior=::NewEnumerator1, Loop Duration=1, UseLoopDelay=false)
   &SpawnBurst_Instantaneous(
    Spawn Count=&RandomRangeInt(Minimum=3, Maximum=5)
    Use Spawn Probability=false
    UseLoopCountLimit=false
   )
  ]
  particleSpawnScript [
   &InitializeParticle(
    Color=&MergeHelper_ColorSwitch(
     0=<rgba 10 5.5 1 1>
     1=<rgba 2 0.6042 1.1868 1>
     2=<rgba 10 5.5 1 1>
     3=<rgba 0.2 1.0139 2 1>
     4=<rgba 0.2 1.55 2 1>
     5=<rgba 0 0 0 0>
     Index=&MakeIntFromFloat(Float=$User.VariantPicker, Conversion Mode=::NewEnumerator1)
     Num=::NewEnumerator4
    )
    Lifetime Max=0.6
    Lifetime Min=0.4
    Lifetime Mode=::NewEnumerator1
    Mesh Scale=&Multiply_VectorByFloat(
     Vector=&RandomRangeVector(Minimum=<0.15 0.2 0.05>, Maximum=<0.3 0.4 0.1>)
     Float=&FN_FrontEndFloatModifier(Gameplay=1, Front End=$System.FrontendScale)
    )
    Mesh Scale Mode=::NewEnumerator4
    Position Mode=::NewEnumerator0
    Sprite Rotation Mode=::NewEnumerator2
    Sprite Size Mode=::NewEnumerator3
    Uniform Sprite Size=&Multiply_Float(
     A=&RandomRangeFloat(Minimum=15, Maximum=30)
     B=&FN_FrontEndFloatModifier(Gameplay=1, Front End=0.5)
    )
    Write Material Random=true
   )
   &InitialMeshOrientation(Mesh Axis To Orient=<0 0 1>, Mesh Orientation Mode=::NewEnumerator2, Use Rotation Vector=true)
   &CameraOffset(Camera Offset Amount=25.4846)
   &AddVelocity(Velocity=&FN_FrontEndVector3Modifier(
     Gameplay=&RandomRangeVector(Minimum=<0 -300 -200>, Maximum=<200 300 300>)
     Front End=&RandomRangeVector(Minimum=<-200 -200 -200>, Maximum=<200 200 200>)
    )) {module=AddVelocity001}
  ]
  particleUpdateScript [
   &ParticleState()
   &ScaleSpriteSize(
    Uniform Curve Index=&Add_Float(
     A=&Multiply_Float(A=$Particles.Age, B=&RandomRangeFloat(Minimum=0.8, Maximum=1.5))
     B=&RandomRangeFloat(Minimum=-0.25, Maximum=0.25)
    )
    Uniform Curve Sprite Scale=DataInterface() {propertyValues {Curve {keys [
        {tangentMode=RCTM_User, time=0.2, value=0}
        <at 0.22 1 RCIM_Cubic>
        {tangentMode=RCTM_User, time=0.5, value=0}
        <at 0.68 0>
        <at 0.7 0.8 RCIM_Cubic>
        <at 0.85 0>
       ]}}}
   )
   &UpdateMeshOrientation(
    Rotation Rate=&RandomRangeFloat(Minimum=2, Maximum=3)
    Rotation Vector=&RandomRangeVector(Minimum=<-1 -1 -1>, Maximum=<1 1 1>)
   )
   &ScaleMeshSize(Scale Factor=&VectorFromFloat(Value=&FloatFromCurve(
      FloatCurve=DataInterface() {propertyValues {Curve {keys [
          {arriveTangent=9.0725, interpMode=RCIM_Cubic, leaveTangent=9.0725, tangentMode=RCTM_User, time=0, value=0}
          {arriveTangent=-0.0026, interpMode=RCIM_Cubic, leaveTangent=-0.0026, tangentMode=RCTM_User, time=0.1408, value=0.9847}
          {arriveTangent=-0.0405, interpMode=RCIM_Cubic, leaveTangent=-0.0405, tangentMode=RCTM_User, time=1, value=0}
         ]}}}
      Scale Curve=&UniformAOrBFloat(
       A=2
       B=3.5
       Distribution Weight=&FN_PickFloatDependOnPlatform(TargetPlatform=DataInterface() {propertyValues {Platforms {qualityLevelMask=-29}}}, TargetPlatformActive=0.7, TargetPlatformNoActive=0.3)
       Evaluation Type=::NewEnumerator0
       Recalculate Random Each Loop=true
       Override Seed=false
      )
     )))
   &Drag(Drag=1.5, Use Linear Drag=true, Use Rotational Drag=true)
   &GravityForce(Gravity=&Multiply_VectorByFloat(
     Vector=<0 0 -980>
     Float=&FloatFromCurve(FloatCurve=DataInterface() {propertyValues {Curve {keys [{arriveTangent=2.0611, interpMode=RCIM_Cubic, leaveTangent=2.0611, tangentMode=RCTM_User, time=0, value=0}, <at 1 1 RCIM_Cubic>]}}})
    ))
   /SetVariables(Particles.ColorMesh:LinearColor=&MergeHelper_ColorSwitch(
     0=&ColorFromCurve(ColorCurve=DataInterface() {propertyValues {
        AlphaCurve {keys [<at 0.0873 1 RCIM_Cubic>,<at 0.151 0 RCIM_Cubic>]}
        BlueCurve {keys [<at 0 3.1789 RCIM_Cubic>,<at 0.6619 1 RCIM_Cubic>]}
        GreenCurve {keys [<at 0 2.1875 RCIM_Cubic>,<at 0.6619 1 RCIM_Cubic>]}
        RedCurve {keys [<at 0 20 RCIM_Cubic>,<at 0.6619 1 RCIM_Cubic>]}
       }})
     1=&ColorFromCurve(ColorCurve=DataInterface() {propertyValues {
        AlphaCurve {keys [<at 0.0841 1 RCIM_Cubic>,<at 0.1608 0 RCIM_Cubic>]}
        BlueCurve {keys [<at 0 10 RCIM_Cubic>,<at 0.6032 2 RCIM_Cubic>]}
        GreenCurve {keys [<at 0 10 RCIM_Cubic>,<at 0.6032 2 RCIM_Cubic>]}
        RedCurve {keys [<at 0 10 RCIM_Cubic>,<at 0.6032 2 RCIM_Cubic>]}
       }})
     2=&ColorFromCurve(ColorCurve=DataInterface() {propertyValues {
        AlphaCurve {keys [<at 0.0891 1 RCIM_Cubic>,<at 0.1607 0 RCIM_Cubic>]}
        BlueCurve {keys [<at 0 1.7188 RCIM_Cubic>,<at 0.6652 0.0548 RCIM_Cubic>]}
        GreenCurve {keys [<at 0 10.0795 RCIM_Cubic>,<at 0.6652 2.9949 RCIM_Cubic>]}
        RedCurve {keys [<at 0 30 RCIM_Cubic>,<at 0.6652 10 RCIM_Cubic>]}
       }})
     3=&ColorFromCurve(ColorCurve=DataInterface() {propertyValues {
        AlphaCurve {keys [<at 0.0891 1 RCIM_Cubic>,<at 0.1607 0 RCIM_Cubic>]}
        BlueCurve {keys [<at 0 30 RCIM_Cubic>,<at 0.6652 10 RCIM_Cubic>]}
        GreenCurve {keys [<at 0 15.5954 RCIM_Cubic>,<at 0.6652 3.5141 RCIM_Cubic>]}
        RedCurve {keys [<at 0 0.9375 RCIM_Cubic>,<at 0.6652 0.0548 RCIM_Cubic>]}
       }})
     4=&RandomRangeLinearColor(
      Minimum=&ColorFromCurve(ColorCurve=DataInterface() {propertyValues {
         AlphaCurve {keys [<at 0.0891 1 RCIM_Cubic>,<at 0.1607 0 RCIM_Cubic>]}
         BlueCurve {keys [<at 0.4285 30 RCIM_Cubic>,<at 0.6652 8.5889 RCIM_Cubic>]}
         GreenCurve {keys [<at 0.4285 29.2421 RCIM_Cubic>,<at 0.6652 1.4583 RCIM_Cubic>]}
         RedCurve {keys [<at 0.4285 0.9375 RCIM_Cubic>,<at 0.6652 10 RCIM_Cubic>]}
        }})
      Maximum=&ColorFromCurve(ColorCurve=DataInterface() {propertyValues {
         AlphaCurve {keys [<at 0.0891 1 RCIM_Cubic>,<at 0.1607 0 RCIM_Cubic>]}
         BlueCurve {keys [<at 0.2694 30 RCIM_Cubic>,<at 0.6594 10>]}
         GreenCurve {keys [<at 0.2694 29.2421 RCIM_Cubic>,<at 0.6594 9.7392>]}
         RedCurve {keys [<at 0.2694 0.9375 RCIM_Cubic>,<at 0.6594 0>]}
        }})
     )
     5=&ColorFromCurve(ColorCurve=DataInterface() {propertyValues {
        AlphaCurve {keys [<at 0.0891 1 RCIM_Cubic>,<at 0.1607 0 RCIM_Cubic>]}
        BlueCurve {keys [<at 0.0042 27.927 RCIM_Cubic>,<at 0.7784 5>]}
        GreenCurve {keys [<at 0.0042 14.3971 RCIM_Cubic>,<at 0.7784 0>]}
        RedCurve {keys [<at 0.0042 30 RCIM_Cubic>,<at 0.7784 0.646>]}
       }})
     Index=&MakeIntFromFloat(Float=$User.VariantPicker, Conversion Mode=::NewEnumerator1)
     Num=::NewEnumerator4
    )) {isSetParameters=true}
   &SolveForcesAndVelocity(Clamp Velocity=false, Limit Acceleration=false)
  ]
  rendererClasses [&NiagaraMeshRendererProperties, &NiagaraSpriteRendererProperties]
  renderers [
   {rendererClass=&NiagaraMeshRendererProperties, rendererIndex=0}
   {rendererClass=&NiagaraSpriteRendererProperties, rendererIndex=1}
  ]
 }
 DarkFlare {
  emitterUpdateScript [
   &EmitterState()
   &SpawnBurst_Instantaneous(Use Spawn Probability=false, UseLoopCountLimit=false)
  ]
  particleSpawnScript [
   &InitializeParticle(
    Lifetime=0.3
    Mass Mode=::NewEnumerator1
    Mesh Scale Mode=::NewEnumerator3
    Mesh Uniform Scale=0.7
    Position Mode=::NewEnumerator1
    Position Offset=$System.HitOffset
    Sprite Rotation Mode=::NewEnumerator2
    Sprite Size=&Vector2DFromFloat(Value=&FN_FrontEndFloatModifier(Gameplay=66, Front End=&Multiply_Float(A=33, B=$System.FrontendScale)))
    Sprite Size Mode=::NewEnumerator4
    UsePositionOffset=true
    Write Material Random=true
   ) {module=InitializeParticle001}
   &CameraOffset(Camera Offset Amount=25)
  ]
  particleUpdateScript [
   &ParticleState()
   &SolveForcesAndVelocity(Clamp Velocity=false, Limit Acceleration=false)
   &ScaleColor(
    Scale Alpha=&FloatFromCurve(FloatCurve=DataInterface() {propertyValues {Curve {keys [
         {arriveTangent=6.4019, interpMode=RCIM_Cubic, leaveTangent=6.4019, tangentMode=RCTM_User, time=0, value=0}
         {arriveTangent=0.002, interpMode=RCIM_Cubic, leaveTangent=0.002, time=0.2, value=0.7}
         {arriveTangent=-0.9006, interpMode=RCIM_Cubic, leaveTangent=-0.9006, tangentMode=RCTM_User, time=1.0011, value=0.002}
        ]}}})
    ScaleA=true
    ScaleRGB=false
   )
   &ScaleSpriteSize(Uniform Curve Sprite Scale=DataInterface() {propertyValues {Curve {keys [
        {arriveTangent=0.7585, interpMode=RCIM_Cubic, leaveTangent=0.7585, tangentMode=RCTM_User, time=0, value=0.4}
        {interpMode=RCIM_Cubic, tangentMode=RCTM_User, time=1, value=1}
       ]}}})
   &SpriteRotationRate(Rotation Rate=&UniformRangedFloat(Minimum=-50, Maximum=50))
   /SetVariables(Particles.DarkGlowSize:Vector2f=&Multiply_Vector2D_ByFloat(Float=1.5, Vector2D=$Particles.SpriteSize)) {isSetParameters=true}
   /SetVariables(Particles.DargGlowColor:LinearColor=&MergeHelper_ColorSwitch(
     0=&ColorFromCurve(ColorCurve=DataInterface() {propertyValues {
        AlphaCurve {keys [<at 0 1 RCIM_Cubic>,<at 1 0 RCIM_Cubic>]}
        BlueCurve {keys [<at 0 0.025 RCIM_Cubic>,<at 1 0.025 RCIM_Cubic>]}
        GreenCurve {keys [<at 0 0 RCIM_Cubic>,<at 1 0 RCIM_Cubic>]}
        RedCurve {keys [<at 0 0.1875 RCIM_Cubic>,<at 1 0.1875 RCIM_Cubic>]}
       }})
     1=&ColorFromCurve(ColorCurve=DataInterface() {propertyValues {
        AlphaCurve {keys [<at 0 1 RCIM_Cubic>,<at 1 0 RCIM_Cubic>]}
        BlueCurve {keys [<at 0 0.0516 RCIM_Cubic>]}
        GreenCurve {keys [<at 0 0 RCIM_Cubic>]}
        RedCurve {keys [<at 0 0.099 RCIM_Cubic>]}
       }})
     2=&ColorFromCurve(ColorCurve=DataInterface() {propertyValues {
        AlphaCurve {keys [<at 0 1 RCIM_Cubic>,<at 1 0 RCIM_Cubic>]}
        BlueCurve {keys [<at 0 0.0001 RCIM_Cubic>]}
        GreenCurve {keys [<at 0 0.0001 RCIM_Cubic>]}
        RedCurve {keys [<at 0 0.174 RCIM_Cubic>]}
       }})
     3=&ColorFromCurve(ColorCurve=DataInterface() {propertyValues {
        AlphaCurve {keys [<at 0 1 RCIM_Cubic>,<at 1 0 RCIM_Cubic>]}
        BlueCurve {keys [<at 0 0.0677 RCIM_Cubic>]}
        GreenCurve {keys [<at 0 0.011 RCIM_Cubic>]}
        RedCurve {keys [<at 0 0 RCIM_Cubic>]}
       }})
     4=&ColorFromCurve(ColorCurve=DataInterface() {propertyValues {
        AlphaCurve {keys [<at 0 1 RCIM_Cubic>,<at 1 0 RCIM_Cubic>]}
        BlueCurve {keys [<at 0 0.0677 RCIM_Cubic>]}
        GreenCurve {keys [<at 0 0.0293 RCIM_Cubic>]}
        RedCurve {keys [<at 0 0 RCIM_Cubic>]}
       }})
     5=&ColorFromCurve(ColorCurve=DataInterface() {propertyValues {
        AlphaCurve {keys [<at 0 1 RCIM_Cubic>,<at 1 0 RCIM_Cubic>]}
        BlueCurve {keys [<at 0 0.2748 RCIM_Cubic>]}
        GreenCurve {keys [<at 0 0.015 RCIM_Cubic>]}
        RedCurve {keys [<at 0 0.2917 RCIM_Cubic>]}
       }})
     Index=&MakeIntFromFloat(Float=$User.VariantPicker, Conversion Mode=::NewEnumerator1)
     Num=::NewEnumerator4
    )) {isSetParameters=true}
   /SetVariables() {isSetParameters=true}
   /SetVariables(Particles.CentralFlareSize:Vector2f=&Multiply_Vector2D_ByFloat(Float=3, Vector2D=$Particles.SpriteSize)) {isSetParameters=true}
   /SetVariables(Particles.CentralFlareColor:LinearColor=&MultiplyLinearColor(
     Linear Color=&MergeHelper_ColorSwitch(
      0=&ColorFromCurve(ColorCurve=DataInterface() {propertyValues {
         AlphaCurve {keys [<at 0 0.8 RCIM_Cubic>,<at 0.6086 0 RCIM_Cubic>]}
         BlueCurve {keys [<at 0.3777 0.5853>,<at 0.9328 2.8793 RCIM_Cubic>]}
         GreenCurve {keys [<at 0.3777 0.012>,<at 0.9328 1.8256 RCIM_Cubic>]}
         RedCurve {keys [<at 0.3777 10>,<at 0.9328 10 RCIM_Cubic>]}
        }})
      1=&ColorFromCurve(ColorCurve=DataInterface() {propertyValues {
         AlphaCurve {keys [<at 0 0.8 RCIM_Cubic>,<at 0.6086 0 RCIM_Cubic>]}
         BlueCurve {keys [<at 0.3777 0.5853>,<at 0.9328 2.8793 RCIM_Cubic>]}
         GreenCurve {keys [<at 0.3777 0.012>,<at 0.9328 1.8256 RCIM_Cubic>]}
         RedCurve {keys [<at 0.3777 10>,<at 0.9328 10 RCIM_Cubic>]}
        }})
      2=&ColorFromCurve(ColorCurve=DataInterface() {propertyValues {
         AlphaCurve {keys [<at 0 0.8 RCIM_Cubic>,<at 0.6086 0 RCIM_Cubic>]}
         BlueCurve {keys [<at 0.3777 0.012>,<at 0.9328 1.8258 RCIM_Cubic>]}
         GreenCurve {keys [<at 0.3777 1.0542>,<at 0.9328 1.8257 RCIM_Cubic>]}
         RedCurve {keys [<at 0.3777 10>,<at 0.9328 10 RCIM_Cubic>]}
        }})
      3=&ColorFromCurve(ColorCurve=DataInterface() {propertyValues {
         AlphaCurve {keys [<at 0 0.8 RCIM_Cubic>,<at 0.6086 0 RCIM_Cubic>]}
         BlueCurve {keys [<at 0.3777 5>,<at 0.9328 10 RCIM_Cubic>]}
         GreenCurve {keys [<at 0.3777 0.7009>,<at 0.9328 2.1101 RCIM_Cubic>]}
         RedCurve {keys [<at 0.3777 0.006>,<at 0.9328 1.8257 RCIM_Cubic>]}
        }})
      4=&ColorFromCurve(ColorCurve=DataInterface() {propertyValues {
         AlphaCurve {keys [<at 0 0.8 RCIM_Cubic>,<at 0.6086 0 RCIM_Cubic>]}
         BlueCurve {keys [<at 0.3777 7.1287>,<at 0.9328 8 RCIM_Cubic>]}
         GreenCurve {keys [<at 0.3777 0.2917>,<at 0.9328 5.2935 RCIM_Cubic>]}
         RedCurve {keys [<at 0.3777 8>,<at 0.9328 2.2729 RCIM_Cubic>]}
        }})
      5=&ColorFromCurve(ColorCurve=DataInterface() {propertyValues {
         AlphaCurve {keys [<at 0 0.8 RCIM_Cubic>,<at 0.6086 0 RCIM_Cubic>]}
         BlueCurve {keys [<at 0.3777 5>,<at 0.9328 6 RCIM_Cubic>]}
         GreenCurve {keys [<at 0.3777 1.8229>,<at 0.9328 0.9391 RCIM_Cubic>]}
         RedCurve {keys [<at 0.3777 4.8618>,<at 0.9328 1.0271 RCIM_Cubic>]}
        }})
      Index=&MakeIntFromFloat(Float=$User.VariantPicker, Conversion Mode=::NewEnumerator1)
      Num=::NewEnumerator4
     )
     Scale Factor=&MakeVector4_FromLinearColor(LinearColor=&FNTimeOfDayColor(Night Color=<rgba 0.3555 0.4043 0.4948 1>, Morning Color=<rgba 1 1 1 1>, Day Color=<rgba 1 1 1 1>, Evening Color=<rgba 1 1 1 1>))
    )) {isSetParameters=true}
  ]
  rendererClasses [&NiagaraSpriteRendererProperties]
  renderers [
   {rendererClass=&NiagaraSpriteRendererProperties}
   {rendererClass=&NiagaraSpriteRendererProperties, rendererIndex=1}
  ]
 }
 Flare_Complex {
  emitterUpdateScript [
   &EmitterState()
   &SpawnBurst_Instantaneous(Use Spawn Probability=false, UseLoopCountLimit=false)
  ]
  particleSpawnScript [
   &InitializeParticle(
    Color=&MergeHelper_ColorSwitch(
     0=<rgba 10 6.8487 4.8958 1>
     1=<rgba 9.8175 6.35 10 1>
     2=<rgba 10 9.9556 4.8958 1>
     3=<rgba 4.6957 9.0314 10 1>
     4=<rgba 8.3333 9.9444 10 1>
     5=<rgba 15 5.0781 13.5477 1>
     Index=&MakeIntFromFloat(Float=$User.VariantPicker, Conversion Mode=::NewEnumerator1)
     Num=::NewEnumerator4
    )
    Lifetime=0.33
    Sprite Rotation Mode=::NewEnumerator2
    Sprite Size Mode=::NewEnumerator3
    Uniform Sprite Size=&Multiply_Float(A=22, B=&FN_FrontEndFloatModifier(Gameplay=1, Front End=$System.FrontendScale))
    UsePositionOffset=false
    Write Material Random=true
   )
   /SetVariables(Particles.DistanceLerp:NiagaraFloat=&FloatFromCurve(FloatCurve=DataInterface() {propertyValues {Curve {keys [<at 500 0>,<at 7000 1>]}}}, CurveIndex=$Engine.Owner.LODDistance)) {isSetParameters=true}
   &CameraOffset(Camera Offset Amount=50) {module=CameraOffset001}
  ]
  particleUpdateScript [
   &ParticleState()
   &ScaleSpriteSize(Uniform Curve Sprite Scale=DataInterface() {propertyValues {Curve {keys [
        {arriveTangent=5.2794, interpMode=RCIM_Cubic, leaveTangent=5.2794, tangentMode=RCTM_User, time=0, value=1}
        {arriveTangent=0.7472, interpMode=RCIM_Cubic, leaveTangent=0.7472, tangentMode=RCTM_User, time=1, value=4}
       ]}}})
   &CameraOffset(Camera Offset Amount=&Multiply_Float(
     A=&Lerp_Float(A=50, B=100, Alpha=$Particles.NormalizedAge)
     B=&Lerp_Float(A=0.4, B=1, Alpha=$User.SizeNormalized)
    ))
   &DynamicMaterialParameters(
    Index 0 Param 1=1
    Index 0 Param 2=&Lerp_Float(
     A=0.7
     B=0
     Alpha=&FloatFromCurve(FloatCurve=DataInterface() {propertyValues {Curve {keys [
          {arriveTangent=1.3124, interpMode=RCIM_Cubic, leaveTangent=1.3124, tangentMode=RCTM_User, time=0, value=0}
          {arriveTangent=0.3903, interpMode=RCIM_Cubic, leaveTangent=0.3903, tangentMode=RCTM_User, time=1, value=1}
         ]}}})
    )
    Module.Index0Param0Write=true
    Module.Index0Param1Write=true
    Module.Index0Param2Write=true
    Module.Index0Param3Write=false
   )
   &ScaleColor(
    Scale Alpha=&FloatFromCurve(
     FloatCurve=DataInterface() {propertyValues {Curve {
        defaultValue=1
        keys [
         {arriveTangent=-2.2905, interpMode=RCIM_Cubic, leaveTangent=-2.2905, tangentMode=RCTM_User, time=0, value=1}
         {arriveTangent=-0.0385, interpMode=RCIM_Cubic, leaveTangent=-0.0385, tangentMode=RCTM_User, time=1, value=0}
        ]
       }}}
     Scale Curve=0.5
    )
    ScaleA=true
    ScaleRGB=true
   )
   /SetVariables(
    Particles.Darken Color:LinearColor=&MakeLinearColorFromVectorAndFloat(
     "Vector (RGB)"=&MakeVectorFromLinearColorRGB(Color=&MergeHelper_ColorSwitch(
       0=<rgba 0.1875 0 0.025 1.2>
       1=<rgba 0.1739 0 0.1361 1>
       2=<rgba 0.174 0.0001 0.0001 1.2>
       3=<rgba 0 0.0277 0.17 2>
       4=<rgba 0 0.08 0.2 1>
       5=<rgba 0.0277 0 0.0677 2>
       Index=&MakeIntFromFloat(Float=$User.VariantPicker, Conversion Mode=::NewEnumerator1)
       Num=::NewEnumerator4
      ))
     "Float (Alpha)"=&FloatFromCurve(
      FloatCurve=DataInterface() {propertyValues {Curve {keys [
          {arriveTangent=-0.5522, interpMode=RCIM_Cubic, leaveTangent=-0.5522, tangentMode=RCTM_User, time=0, value=1}
          {interpMode=RCIM_Cubic, tangentMode=RCTM_User, time=1, value=0}
         ]}}}
      Scale Curve=&MergeHelper_FloatSwitch(
       0=1.5
       1=1.2
       2=1
       3=1
       4=1.8
       5=2.5
       Index=&MakeIntFromFloat(Float=$User.VariantPicker, Conversion Mode=::NewEnumerator1)
       Num=::NewEnumerator4
      )
     )
    )
    Particles.Darken Size:Vector2f=&Multiply_Vector2D_ByFloat(Float=1.4, Vector2D=$Particles.SpriteSize)
   ) {isSetParameters=true}
   /SetVariables(
    Particles.Glow Color=&MultiplyLinearColor(
     Linear Color=$Particles.Color
     Scale Factor=&MakeFloat4FromVectorAndFloat(
      "Vector (RGB)"=<1 1 1>
      "Float (Alpha)"=&FloatFromCurve(
       FloatCurve=DataInterface() {propertyValues {Curve {keys [
           {arriveTangent=3.5878, interpMode=RCIM_Cubic, leaveTangent=3.5878, tangentMode=RCTM_User, time=0, value=0.35}
           {arriveTangent=0.3, interpMode=RCIM_Cubic, leaveTangent=0.3, time=0.3, value=1}
           {arriveTangent=3.4933, interpMode=RCIM_Cubic, leaveTangent=3.4933, tangentMode=RCTM_User, time=0.5, value=0.5}
           {arriveTangent=-0.4, arriveTangentWeight=0.0807, interpMode=RCIM_Cubic, leaveTangent=-0.4, leaveTangentWeight=0.3549, time=0.6, value=0.75}
           {arriveTangent=-0.0405, interpMode=RCIM_Cubic, leaveTangent=-0.0405, tangentMode=RCTM_User, time=1, value=0.3}
          ]}}}
       Scale Curve=0.05
      )
     )
    )
    Particles.Glow Size=&Vector2DFromFloat(Value=&FN_FrontEndFloatModifier(Gameplay=120, Front End=20))
   ) {isSetParameters=true}
   /SetVariables(Particles.Wide Flare Scale:Vector2f=&Vector2DFromCurve(
     Vector2Curve=DataInterface() {propertyValues {
       XCurve {keys [
         {arriveTangent=4.673, interpMode=RCIM_Cubic, leaveTangent=4.673, tangentMode=RCTM_User, time=0, value=0}
         {arriveTangentWeight=0.1536, interpMode=RCIM_Cubic, leaveTangentWeight=0.1152, tangentMode=RCTM_User, time=0.3, value=1}
         {arriveTangent=-3.1914, interpMode=RCIM_Cubic, leaveTangent=-3.1914, tangentMode=RCTM_User, time=0.5, value=0.15}
         {arriveTangent=-2.3891, interpMode=RCIM_Cubic, leaveTangent=-2.3891, tangentMode=RCTM_User, time=0.7, value=0.25}
         {interpMode=RCIM_Cubic, leaveTangent=-1.231, time=1, value=0}
        ]}
       YCurve {keys [
         {arriveTangent=10.6903, interpMode=RCIM_Cubic, leaveTangent=10.6903, tangentMode=RCTM_User, time=0, value=0}
         {arriveTangent=0.8, interpMode=RCIM_Cubic, leaveTangent=0.8, time=0.25, value=1}
         {arriveTangent=-1.7663, interpMode=RCIM_Cubic, leaveTangent=-1.7663, tangentMode=RCTM_User, time=0.5, value=0.4}
         {arriveTangent=-4.8806, interpMode=RCIM_Cubic, leaveTangent=-4.8806, tangentMode=RCTM_User, time=0.7, value=1}
         {interpMode=RCIM_Cubic, leaveTangent=-2.2148, time=1, value=0}
        ]}
      }}
     ScaleCurve=&Multiply_Vector2D_ByFloat(Float=3, Vector2D=$Particles.SpriteSize)
    )) {isSetParameters=true}
  ]
  rendererClasses [&NiagaraSpriteRendererProperties]
  renderers [
   {rendererClass=&NiagaraSpriteRendererProperties}
   {rendererClass=&NiagaraSpriteRendererProperties, rendererIndex=1}
   {rendererClass=&NiagaraSpriteRendererProperties, rendererIndex=2}
  ]
 }
 Flash_Elements {
  emitterUpdateScript [
   &EmitterState()
   &SpawnBurst_Instantaneous(Spawn Group=1, Use Spawn Probability=false, UseLoopCountLimit=false) {module=SpawnBurst_Instantaneous001}
   &SpawnBurst_Instantaneous(
    Spawn Count=&RandomRangeInt(Minimum=6, Maximum=8)
    Use Spawn Probability=false
    UseLoopCountLimit=false
   )
  ]
  particleSpawnScript [
   &InitializeParticle(Color Mode=::NewEnumerator0, Lifetime=0.25, UsePositionOffset=false, Write Material Random=true)
   &ShapeLocation(MaskSpawnGroup=false, Override Seed=false, Sphere Radius=1, UseOwnerScale=true)
   &SpriteFacingAndAlignment(Sprite Alignment=$Particles.ShapeLocation.ShapeNormal, WriteAlignment=true, WriteFacing=false)
   &FN_PerParticleLOD(Enable SpriteSize Distance Scaling=true, Sprite Size Scale Far=<2.5 2.5>)
   /SetVariables(Particles.SpriteSize=<1 1>, Particles.VisibilityTag=&SpawnGroup()) {isSetParameters=true}
   /SetVariables(
    Particles.Flash Spike Color:LinearColor=&MergeHelper_ColorSwitch(
     0=<rgba 1 0.0504 0.2395 1>
     1=<rgba 1 0.0504 0.7936 1>
     2=<rgba 1 0.2995 0.0055 1>
     3=<rgba 0.0104 0.4063 1 1>
     4=<rgba 0.4072 0.855 0.8879 1>
     5=<rgba 0.3051 0.7013 0.9047 1>
     Index=&MakeIntFromFloat(Float=$User.VariantPicker, Conversion Mode=::NewEnumerator1)
     Num=::NewEnumerator4
    )
    Particles.Flash Spike Lifetime:NiagaraFloat=&RandomRangeFloat(Minimum=0.08, Maximum=0.12)
    Particles.Flash Spike Sprite Size:Vector2f=&FN_FrontEndVector2Modifier(
     Front End:Vector2f=&RandomRangeVector2D(Minimum=<2 25>, Maximum=<7 40>)
     Gameplay:Vector2f=&RandomRangeVector2D(Minimum=<3 55>, Maximum=<10 70>)
    )
   ) {isSetParameters=true}
   /SetVariables(
    Particles.Flash Star Cam Offset=15
    Particles.Flash Star Lifetime=0.15
    Particles.Flash Star Rotation:NiagaraFloat=&RandomRangeFloat(Minimum=0, Maximum=360)
    Particles.Flash Star Sprite Size:Vector2f=&Multiply_Vector2D_ByFloat(
     Float=&Multiply_Float(
      A=&MakeCustomFloatFromBool(Bool=$User.bIsFrontEnd, True Float=0.45, False Float=1)
      B=&RandomRangeFloat(Minimum=120, Maximum=150)
     )
     Vector2D=$Particles.SpriteSize
    )
    Particles.Flash Star Sprite Size Alt:Vector2f=&Multiply_Vector2D_ByFloat(
     Float=&Multiply_Float(
      A=0.85
      B=&Multiply_Float(
       A=&MakeCustomFloatFromBool(Bool=$User.bIsFrontEnd, True Float=0.45, False Float=1)
       B=&RandomRangeFloat(Minimum=120, Maximum=150)
      )
     )
     Vector2D=$Particles.SpriteSize
    )
    Particles.Flash Star SubUV Index:NiagaraFloat=&RandomRangeFloat(Minimum=0, Maximum=3)
    Particles.Flash Star SubUV Index Alt:NiagaraFloat=&RandomRangeFloat(Minimum=0, Maximum=3)
   ) {isSetParameters=true}
   /SetVariables(
    Particles.Flash Flare Color:LinearColor=&MergeHelper_ColorSwitch(
     0=<rgba 1 0.006 0.1146 0.7>
     1=<rgba 0.9473 0.4072 0.7913 0.7>
     2=<rgba 1 0.855 0.2789 0.7>
     3=<rgba 0.2789 0.8683 1 0.7>
     4=<rgba 0.1458 0.8886 1 0.7>
     5=<rgba 0.9301 0.4064 0.9004 0.7>
     Index=&MakeIntFromFloat(Float=$User.VariantPicker, Conversion Mode=::NewEnumerator1)
     Num=::NewEnumerator4
    )
    Particles.Flash Flare Lifetime:NiagaraFloat=&FN_FrontEndFloatModifier(Gameplay=0.2, Front End=0.125)
    Particles.Flash Flare Sprite Size:Vector2f=&FN_FrontEndVector2Modifier(Front End:Vector2f=<90 90>, Gameplay:Vector2f=<130 130>)
   ) {isSetParameters=true}
   &DynamicMaterialParameters(
    Index 0 Param 1=0.2
    Index 0 Param 2=1
    Module.Index0Float0Write=true
    Module.Index0Float1Write=true
    Module.Index0Float2Write=true
    Module.Index0Float3Write=true
   )
  ]
  particleUpdateScript [
   &ParticleState()
   /SetVariables(
    Particles.Update Flash Flare Age:NiagaraFloat=&Divide_Float(A=$Particles.Age, B=$Particles.Initial.Flash Flare Lifetime)
    Particles.Update Flash Spike Age:NiagaraFloat=&Divide_Float(A=$Particles.Age, B=$Particles.Initial.Flash Spike Lifetime)
    Particles.Update Flash Star Age:NiagaraFloat=&Divide_Float(A=$Particles.Age, B=$Particles.Initial.Flash Star Lifetime)
   ) {isSetParameters=true}
   /SetVariables(
    Particles.Update Flash Spike Color:LinearColor=&MakeLinearColorFromVectorAndFloat(
     "Vector (RGB)"=&MakeVectorFromLinearColorRGB(Color=$Particles.Initial.Flash Spike Color)
     "Float (Alpha)"=&FloatFromCurve(FloatCurve=DataInterface() {propertyValues {Curve {keys [<at 0.7 1>,<at 1 0>]}}}, CurveIndex=$Particles.Update Flash Spike Age)
    )
    Particles.Update Flash Spike Sprite Size:Vector2f=&Multiply_Vector2D_ByFloat(
     Float=&FloatFromCurve(FloatCurve=DataInterface() {propertyValues {Curve {keys [<at 0 0.6>,<at 1 2>]}}}, CurveIndex=$Particles.Update Flash Spike Age)
     Vector2D=$Particles.Initial.Flash Spike Sprite Size
    )
    Particles.Update Flash Star Color:LinearColor=&MergeHelper_ColorSwitch(
     0=&ColorFromCurve(
      ColorCurve=DataInterface() {propertyValues {
        AlphaCurve {keys [<at 0 1>,<at 0.8992 0>]}
        BlueCurve {keys [<at 0.1978 1.964>,<at 0.4068 0.9078>,<at 0.6127 0.4041>]}
        GreenCurve {keys [<at 0.1978 0.4486>,<at 0.4068 0.258>,<at 0.6127 0.0456>]}
        RedCurve {keys [<at 0.1978 5>,<at 0.4068 3>,<at 0.6127 1.5>]}
       }}
      CurveIndex=$Particles.Update Flash Star Age
     )
     1=&ColorFromCurve(
      ColorCurve=DataInterface() {propertyValues {
        AlphaCurve {keys [<at 0 1>,<at 0.8992 0>]}
        BlueCurve {keys [<at 0.1978 4.1766>,<at 0.4083 2.1775>,<at 0.6127 1.0321>]}
        GreenCurve {keys [<at 0.1978 2.1495>,<at 0.4083 0.7726>,<at 0.6127 0.0456>]}
        RedCurve {keys [<at 0.1978 5>,<at 0.4083 3>,<at 0.6127 1.5>]}
       }}
      CurveIndex=$Particles.Update Flash Star Age
     )
     2=&ColorFromCurve(
      ColorCurve=DataInterface() {propertyValues {
        AlphaCurve {keys [<at 0 1>,<at 0.8992 0>]}
        BlueCurve {keys [<at 0.1978 0.5>,<at 0.4083 0.3>,<at 0.6127 0>]}
        GreenCurve {keys [<at 0.1978 2.75>,<at 0.4083 0.5817>,<at 0.6127 0.0563>]}
        RedCurve {keys [<at 0.1978 5>,<at 0.4083 3>,<at 0.6127 0.5391>]}
       }}
      CurveIndex=$Particles.Update Flash Star Age
     )
     3=&ColorFromCurve(
      ColorCurve=DataInterface() {propertyValues {
        AlphaCurve {keys [<at 0 1>,<at 0.8992 0>]}
        BlueCurve {keys [<at 0.1978 5>,<at 0.4068 3>,<at 0.6127 0.5391>]}
        GreenCurve {keys [<at 0.1978 4.3416>,<at 0.4068 1.0542>,<at 0.6127 0.1594>]}
        RedCurve {keys [<at 0.1978 1.3945>,<at 0.4068 0.0164>,<at 0.6127 0>]}
       }}
      CurveIndex=$Particles.Update Flash Star Age
     )
     4=&ColorFromCurve(
      ColorCurve=DataInterface() {propertyValues {
        AlphaCurve {keys [<at 0 1>,<at 0.8992 0>]}
        BlueCurve {keys [<at 0.1547 5>,<at 0.3137 3>,<at 0.6127 0.3281>]}
        GreenCurve {keys [<at 0.1547 4.8616>,<at 0.3137 2.7665>,<at 0.6127 0.1887>]}
        RedCurve {keys [<at 0.1547 2.3394>,<at 0.3137 0.0164>,<at 0.6127 0.0137>]}
       }}
      CurveIndex=$Particles.Update Flash Star Age
     )
     5=&ColorFromCurve(
      ColorCurve=DataInterface() {propertyValues {
        AlphaCurve {keys [<at 0 1>,<at 0.8992 0>]}
        BlueCurve {keys [<at 0.1547 5>,<at 0.2999 3>,<at 0.6179 0.487>]}
        GreenCurve {keys [<at 0.1547 4.0977>,<at 0.2999 1.822>,<at 0.6179 0.2248>]}
        RedCurve {keys [<at 0.1547 2.3394>,<at 0.2999 0.3437>,<at 0.6179 0.0085>]}
       }}
      CurveIndex=$Particles.Update Flash Star Age
     )
     Index=&MakeIntFromFloat(Float=$User.VariantPicker, Conversion Mode=::NewEnumerator1)
     Num=::NewEnumerator4
    )
    Particles.Update Flash Star Sprize Size:Vector2f=&Multiply_Vector2D_ByFloat(
     Float=&FloatFromCurve(
      FloatCurve=DataInterface() {propertyValues {Curve {keys [
          {arriveTangent=2.7013, interpMode=RCIM_Cubic, leaveTangent=2.7013, tangentMode=RCTM_User, time=0, value=0}
          {interpMode=RCIM_Cubic, tangentMode=RCTM_User, time=1, value=1}
         ]}}}
      CurveIndex=$Particles.Update Flash Star Age
     )
     Vector2D=$Particles.Initial.Flash Star Sprite Size
    )
    Particles.Update Flash Star Sprize Size Alt:Vector2f=&Multiply_Vector2D_ByFloat(
     Float=&FloatFromCurve(
      FloatCurve=DataInterface() {propertyValues {Curve {keys [
          {arriveTangent=2.7013, interpMode=RCIM_Cubic, leaveTangent=2.7013, tangentMode=RCTM_User, time=0, value=0}
          {interpMode=RCIM_Cubic, tangentMode=RCTM_User, time=1, value=1}
         ]}}}
      CurveIndex=$Particles.Update Flash Star Age
     )
     Vector2D=$Particles.Initial.Flash Star Sprite Size Alt
    )
   ) {isSetParameters=true}
   /SetVariables(Particles.Update Flash Flare Color:LinearColor=&MakeLinearColorFromVectorAndFloat(
     "Vector (RGB)"=&Multiply_Vector(
      A=&MakeVectorFromLinearColorRGB(Color=$Particles.Initial.Flash Flare Color)
      B=&VectorFromFloat(Value=&FN_FrontEndFloatModifier(Gameplay=1, Front End=1))
     )
     "Float (Alpha)"=&Multiply_Float(
      A=&MakeFloatFromLinearColor(LinearColor=<rgba 1 1 1 1>, Channel:ENiagaraLinearColor_Channels=::NewEnumerator3)
      B=&Lerp_Float(A=1, B=0, Alpha=$Particles.Update Flash Flare Age)
     )
    )) {isSetParameters=true}
   &FN_Light_Attributes(
    Color=&MergeHelper_ColorSwitch(
     0=<rgba 1 0.0821 0.4458 1>
     1=<rgba 0.8388 0.3095 0.7454 1>
     2=<rgba 0.8 0.4338 0.2231 0.9957>
     3=<rgba 0.0055 0.3514 1 0.9362>
     4=<rgba 0.1046 0.624 1 1>
     5=<rgba 0.7822 0.1023 0.8542 1>
     Index=&MakeIntFromFloat(Float=$User.VariantPicker, Conversion Mode=::NewEnumerator1)
     Num=::NewEnumerator4
    )
    Create Color Attribute=true
    Create Exponent Attribute=true
    Create Position Attribute=true
    Create Radius Attribute=true
    Exponent=5
    Position=<15 0 0>
    Radius=90
    Scale Color=&FloatFromCurve(
     FloatCurve=DataInterface() {propertyValues {Curve {keys [
         {interpMode=RCIM_Cubic, tangentMode=RCTM_User, time=0, value=1}
         {arriveTangent=-0.5426, interpMode=RCIM_Cubic, leaveTangent=-0.5426, tangentMode=RCTM_User, time=1, value=0}
        ]}}}
     Scale Curve=10
    )
   )
  ]
  rendererClasses [&NiagaraSpriteRendererProperties, &NiagaraLightRendererProperties]
  renderers [
   {rendererClass=&NiagaraSpriteRendererProperties}
   {rendererClass=&NiagaraSpriteRendererProperties, rendererIndex=1}
   {rendererClass=&NiagaraSpriteRendererProperties, rendererIndex=2}
   {rendererClass=&NiagaraLightRendererProperties, rendererIndex=3}
  ]
 }
 GalaxySmoke {
  emitterUpdateScript [
   &EmitterState(Life Cycle Mode=::NewEnumerator1, Loop Behavior=::NewEnumerator1, Loop Duration=1, UseLoopDelay=false)
   &SpawnBurst_Instantaneous(Use Spawn Probability=false, UseLoopCountLimit=false)
   &SpawnBurst_Instantaneous(Spawn Group=1, Use Spawn Probability=false, UseLoopCountLimit=false) {module=SpawnBurst_Instantaneous001}
  ]
  particleSpawnScript [
   &InitializeParticle(
    Color=<rgba 1 1 1 3>
    Lifetime=&MergeHelper_FloatSwitch(
     0=0.55
     1=&RandomRangeFloat(Minimum=0.3, Maximum=0.4)
     2=?
     3=?
     Index=&SpawnGroup()
     Num=::NewEnumerator0
    )
    Mesh Scale Mode=::NewEnumerator3
    Mesh Uniform Scale=&Multiply_Float(A=0.85, B=&FN_FrontEndFloatModifier(Gameplay=1, Front End=1))
    Sprite Size Mode=::NewEnumerator3
    Uniform Sprite Size=&Multiply_Float(
     A=&RandomRangeFloat(Minimum=85, Maximum=110)
     B=&FN_FrontEndFloatModifier(Gameplay=1, Front End=1)
    )
    UsePositionOffset=false
    Write Material Random=true
   )
   &InitialMeshOrientation(Orientation Axis=<0 0 1>, Orientation Vector=<0 0 1>, Rotation=<0 0.25 0>, Use Orientation Vector=true, Use Rotation Vector=true)
   &InitialMeshOrientation(
    Orientation Axis=<0 0 0>
    Orientation Vector=<0 0 0>
    Rotation=&RandomRangeVector(Minimum=<0 0 0>, Maximum=<1 0 0>)
    Use Orientation Vector=false
    Use Rotation Vector=true
   ) {module=InitialMeshOrientation001}
   &CameraOffset(Camera Offset Amount=30)
  ]
  particleUpdateScript [
   &PartitionParticles()
   &ParticleState()
   &ScaleSpriteSize(
    Uniform Curve Scale=&FN_FrontEndFloatModifier(Gameplay=1, Front End=$System.FrontendScale)
    Uniform Curve Sprite Scale=DataInterface() {propertyValues {Curve {keys [
        {arriveTangent=1.8336, interpMode=RCIM_Cubic, leaveTangent=1.8336, tangentMode=RCTM_User, time=0, value=0}
        {interpMode=RCIM_Cubic, tangentMode=RCTM_User, time=0.9, value=1}
       ]}}}
   )
   &ScaleMeshSize(Scale Factor=&VectorFromFloat(Value=&FloatFromCurve(
      FloatCurve=DataInterface() {propertyValues {Curve {keys [
          {arriveTangent=1.152, interpMode=RCIM_Cubic, leaveTangent=1.152, tangentMode=RCTM_User, time=0, value=0.65}
          {arriveTangent=0.0224, interpMode=RCIM_Cubic, leaveTangent=0.0224, tangentMode=RCTM_User, time=1, value=1}
         ]}}}
      Scale Curve=&FN_FrontEndFloatModifier(Gameplay=1, Front End=$System.FrontendScale)
     )))
   &ScaleColor(
    Scale Alpha=&FloatFromCurve(FloatCurve=DataInterface() {propertyValues {Curve {keys [{interpMode=RCIM_Cubic, tangentMode=RCTM_User, time=0, value=1}, {interpMode=RCIM_Cubic, tangentMode=RCTM_User, time=1, value=0}]}}})
    Scale RGB=&Lerp_Vector(
     A=&MakeVectorFromLinearColorRGB(Color=<rgba 1 0.1 0.418 1>)
     B=&MakeVectorFromLinearColorRGB(Color=<rgba 0 0.2294 0.974 1>)
     Alpha=&FloatFromCurve(FloatCurve=DataInterface() {propertyValues {Curve {keys [{interpMode=RCIM_Cubic, tangentMode=RCTM_User, time=0.4, value=0}, {interpMode=RCIM_Cubic, tangentMode=RCTM_User, time=0.6, value=1}]}}})
    )
    ScaleA=true
    ScaleRGB=true
   ) {module=ScaleColor002}
   &ScaleColor(
    Scale Alpha=&FNTimeOfDayScalar(Night Scale=0.7, Morning Scale=0.5, Evening Scale=0.5)
    Scale RGB=&VectorFromFloat(Value=&FloatFromCurve(
      FloatCurve=DataInterface() {propertyValues {Curve {keys [
          {arriveTangent=11.5613, interpMode=RCIM_Cubic, leaveTangent=11.5613, tangentMode=RCTM_User, time=0, value=0}
          {interpMode=RCIM_Cubic, tangentMode=RCTM_User, time=0.2, value=1}
          {tangentMode=RCTM_User, time=1, value=0}
         ]}}}
      Scale Curve=3
     ))
    ScaleA=true
    ScaleRGB=true
   ) {module=ScaleColor001}
   &DynamicMaterialParameters(
    Index 0 Param 1=&FloatFromCurve(FloatCurve=DataInterface() {propertyValues {Curve {keys [{arriveTangent=-2.1015, interpMode=RCIM_Cubic, leaveTangent=-2.1015, tangentMode=RCTM_User, time=0, value=0.5}, <at 1 -0.15>]}}})
    Index 0 Param 2=&FloatFromCurve(FloatCurve=DataInterface() {propertyValues {Curve {keys [{arriveTangent=-0.3464, interpMode=RCIM_Cubic, leaveTangent=-0.3464, tangentMode=RCTM_User, time=0, value=1}, <at 1 0.9>]}}})
    Index 0 Param 3=&FloatFromCurve(FloatCurve=DataInterface() {propertyValues {Curve {keys [{interpMode=RCIM_Cubic, tangentMode=RCTM_User, time=0, value=1}, {interpMode=RCIM_Cubic, tangentMode=RCTM_User, time=1, value=0.25}]}}})
    Module.Param0WriteEnabled=true
    Module.Param1WriteEnabled=true
    Module.Param2WriteEnabled=true
    Module.Param3WriteEnabled=false
   )
   /SetVariables(Particles.Intensity:NiagaraFloat=&FloatFromCurve(
     FloatCurve=DataInterface() {propertyValues {Curve {keys [{interpMode=RCIM_Cubic, tangentMode=RCTM_User, time=0.2, value=1}, {interpMode=RCIM_Cubic, tangentMode=RCTM_User, time=1, value=0}]}}}
     Scale Curve=100
    )) {isSetParameters=true}
   /SetVariables(Particles.Color Smoke:LinearColor=&MakeLinearColorFromVectorAndFloat(
     "Vector (RGB)"=<1 1 1>
     "Float (Alpha)"=&FloatFromCurve(
      FloatCurve=DataInterface() {propertyValues {Curve {
         defaultValue=1
         keys [
          {arriveTangent=-0.008, interpMode=RCIM_Cubic, leaveTangent=-0.008, tangentMode=RCTM_User, time=0, value=1}
          {arriveTangent=0.1218, interpMode=RCIM_Cubic, leaveTangent=0.1218, tangentMode=RCTM_User, time=0.6485, value=0}
         ]
        }}}
      Scale Curve=3
     )
    )) {isSetParameters=true}
  ]
  rendererClasses [&NiagaraMeshRendererProperties, &NiagaraSpriteRendererProperties]
  renderers [
   {rendererClass=&NiagaraMeshRendererProperties, rendererIndex=0}
   {rendererClass=&NiagaraSpriteRendererProperties, rendererIndex=1}
  ]
 }
 NE_Spark {
  emitterUpdateScript [
   &EmitterState(Life Cycle Mode=::NewEnumerator1, Loop Behavior=::NewEnumerator1, Loop Delay=0.05, Loop Duration=1, UseLoopDelay=true)
   &SpawnBurst_Instantaneous(
    Spawn Count=&RandomRangeInt(Minimum=7, Maximum=12)
    Use Spawn Probability=false
    UseLoopCountLimit=false
   )
  ]
  particleSpawnScript [
   &InitializeParticle(
    Color=&MergeHelper_ColorSwitch(
     0=<rgba 1 0.0822 0.2885 1>
     1=<rgba 1 0.0822 0.2885 1>
     2=<rgba 1 0.2995 0.0055 1>
     3=<rgba 0.251 0.7815 0.9 0.9957>
     4=<rgba 0.2502 0.7835 0.8963 1>
     5=<rgba 0.9 0.4756 0.8475 0.9957>
     Index=&MakeIntFromFloat(Float=$User.VariantPicker, Conversion Mode=::NewEnumerator1)
     Num=::NewEnumerator4
    )
    Lifetime Max=0.5
    Lifetime Min=0.4
    Lifetime Mode=::NewEnumerator1
    Mesh Scale Mode=::NewEnumerator3
    Mesh Uniform Scale=&RandomRangeFloat(Minimum=0.1, Maximum=0.5)
    Sprite Size=&Multiply_Vector2D_ByFloat(
     Float=&FN_FrontEndFloatModifier(Gameplay=1, Front End=$System.FrontendScale)
     Vector2D=&RandomRangeVector2D(Minimum=<2 7>, Maximum=<3 10>)
    )
    Sprite Size Mode=::NewEnumerator4
    UsePositionOffset=false
    Write Material Random=true
   )
   &ShapeLocation(MaskSpawnGroup=false, Override Seed=false, Sphere Radius=10, UseOwnerScale=true)
   &AddVelocityFromPoint(
    Module.Use Falloff Distance=false
    Velocity Strength=&Multiply_Float(
     A=&Multiply_Float(
      A=&RandomRangeFloat(Minimum=150, Maximum=500)
      B=&FN_FrontEndFloatModifier(Gameplay=1, Front End=$System.FrontendScale)
     )
     B=1
    )
   )
   &AddVelocity(Velocity=&RandomRangeVector(Minimum=<0 0 150>, Maximum=<0 0 250>))
  ]
  particleUpdateScript [
   &ParticleState()
   &Drag(Drag=5, Use Linear Drag=true, Use Rotational Drag=true)
   &CurlNoiseForce(Bias Noise=false, Noise Frequency=0.1, Noise Strength=100, Pan Noise=true)
   &GravityForce(Gravity=&RandomRangeVector(Minimum=<0 0 -250>, Maximum=<0 0 -680>))
   &SolveForcesAndVelocity(Clamp Velocity=false, Limit Acceleration=false)
   &ScaleSpriteSizeBySpeed(Max Scale Factor=<1 2>, Min Scale Factor=<1 0.7>)
   &ScaleSpriteSize(Uniform Curve Sprite Scale=DataInterface() {propertyValues {Curve {keys [<at 0 0 RCIM_Cubic>,<at 0.1 1 RCIM_Cubic>,<at 1 0 RCIM_Cubic>]}}})
   /SetVariables(Particles.GlowColor=&MultiplyLinearColor(Linear Color=$Particles.Color, Scale Factor=<1 1 1 0.1>)) {isSetParameters=true}
   /SetVariables(Particles.GlowSize=&Multiply_Vector2D(A=$Particles.SpriteSize, B=<5 2.5>)) {isSetParameters=true}
  ]
  rendererClasses [&NiagaraSpriteRendererProperties]
  renderers [
   {rendererClass=&NiagaraSpriteRendererProperties}
   {rendererClass=&NiagaraSpriteRendererProperties, rendererIndex=1}
  ]
 }
 SnowFlakes {
  emitterUpdateScript [
   &EmitterState(
    Enable Distance Culling=true
    Enable Visibility Culling=true
    Life Cycle Mode=::NewEnumerator1
    Loop Behavior=::NewEnumerator1
    Loop Duration=2
    MaxDistance=2500
    Module.bApplySpawnCountScaleByDistance=true
    Module.bUseMaxDistance=true
    Module.bUseMinDistance=false
    Scalability Mode=::NewEnumerator1
    Scale Spawn Count=true
    UseLoopDelay=false
   )
   &SpawnBurst_Instantaneous(Spawn Count=3, Use Spawn Probability=false, UseLoopCountLimit=false)
  ]
  particleSpawnScript [
   &InitializeParticle(
    Color=&RandomRangeLinearColor(Minimum=<rgba 0.9369 0.9712 1 1>, Maximum=<rgba 0.2865 0.6861 1 1>)
    Lifetime Max=1
    Lifetime Min=0.15
    Lifetime Mode=::NewEnumerator1
    Sprite Rotation Mode=::NewEnumerator2
    Sprite Size Mode=::NewEnumerator3
    Uniform Sprite Size=&FN_FrontEndFloatModifier(Gameplay=&RandomRangeFloat(Minimum=8, Maximum=12), Front End=$System.FrontendScale)
    UsePositionOffset=true
    Write Material Random=true
   )
   &ShapeLocation(MaskSpawnGroup=false, Override Seed=false, Sphere Radius=40, Transform Order:ENiagara_TransformOrder=::NewEnumerator1, UseOwnerScale=true) {module=ShapeLocation001}
   &AddVelocity(
    Module.Use Falloff Distance=false
    Override Seed=false
    Velocity Mode=::NewEnumerator1
    Velocity Speed=&RandomRangeFloat(Minimum=0, Maximum=50)
   ) {module=AddVelocity001}
   &CameraOffset(Camera Offset Amount=25)
  ]
  particleUpdateScript [
   &ParticleState()
   &SpriteRotationRate(Rotation Rate=&RandomRangeFloat(Minimum=-360, Maximum=360))
   &Drag(Drag=10, Use Linear Drag=true, Use Rotational Drag=true)
   &CurlNoiseForce(Noise Strength=200, Pan Noise=false)
   &GravityForce(Gravity=<0 0 -200>)
   &SolveForcesAndVelocity(Clamp Velocity=false, Limit Acceleration=false)
   &ScaleSpriteSize(Uniform Curve Sprite Scale=DataInterface() {propertyValues {Curve {
       defaultValue=0
       keys [{interpMode=RCIM_Cubic, tangentMode=RCTM_User, time=0, value=1}, {interpMode=RCIM_Cubic, tangentMode=RCTM_User, time=1, value=0}]
      }}})
   &ScaleColor(
    Scale Alpha=&FloatFromCurve(FloatCurve=DataInterface() {propertyValues {Curve {
        defaultValue=1
        keys [
         {interpMode=RCIM_Cubic, tangentMode=RCTM_User, time=0, value=1}
         {arriveTangent=-2.6414, interpMode=RCIM_Cubic, leaveTangent=-2.6414, tangentMode=RCTM_User, time=1, value=0}
        ]
       }}})
    ScaleA=true
    ScaleRGB=true
   )
  ]
  rendererClasses [&NiagaraSpriteRendererProperties]
  renderers [{rendererClass=&NiagaraSpriteRendererProperties}]
 }
 Sparkles {
  emitterUpdateScript [
   &EmitterState(Life Cycle Mode=::NewEnumerator1, Loop Behavior=::NewEnumerator1, Loop Delay=0.005, Loop Duration=0.25, UseLoopDelay=false)
   &SpawnRate(SpawnRate=10, Use Spawn Probability=false)
   &SpawnBurst_Instantaneous(Spawn Count=10, Use Spawn Probability=false, UseLoopCountLimit=false)
  ]
  particleSpawnScript [
   &FN_SetSpawnGroup(Number of Spawn Groups=2)
   &InitializeParticle(
    Color=<rgba 0.05 0.05 0.05 0.5>
    Lifetime Max=0.5
    Lifetime Min=0.3
    Lifetime Mode=::NewEnumerator1
    Sprite Rotation Mode=::NewEnumerator2
    Sprite Size Mode=::NewEnumerator1
    Uniform Sprite Size Max=20
    Uniform Sprite Size Min=8
    UsePositionOffset=false
    Write Material Random=true
   )
   &SphereLocation(
    MaskSpawnGroup=false
    Override Seed=false
    Sphere Radius=&Multiply_Float(A=25, B=&FN_FrontEndFloatModifier(Gameplay=1, Front End=$System.FrontendScale))
    Surface Expansion Mode=::NewEnumerator1
    Surface Only=true
    UseNonUniformScale=false
    UseOffset=true
    UseSphereOrientation=false
   )
   &CameraOffset(Camera Offset Amount=4)
   /SetVariables(Particles.VisibilityTag=&SpawnGroup()) {isSetParameters=true}
  ]
  particleUpdateScript [
   &ParticleState()
   &SpriteRotationRate(Rotation Rate=&RandomRangeFloat(Minimum=-360, Maximum=360))
   &SolveForcesAndVelocity(Clamp Velocity=false, Limit Acceleration=false)
   &Color(
    Color=&MergeHelper_ColorSwitch(
     0=&RandomRangeLinearColor(Minimum=<rgba 5 2.75 0.5 0.1>, Maximum=<rgba 2 1.1 0.2 1>)
     1=&RandomRangeLinearColor(Minimum=<rgba 1 0.4323 0.8174 0.9957>, Maximum=<rgba 2 0.6042 1.1868 1>)
     2=&RandomRangeLinearColor(Minimum=<rgba 5 2.75 0.5 0.1>, Maximum=<rgba 2 0.011 0.011 1>)
     3=&RandomRangeLinearColor(Minimum=<rgba 0.9323 0.977 1 0.9957>, Maximum=<rgba 0.2 1.0139 2 1>)
     4=&RandomRangeLinearColor(Minimum=<rgba 1.7691 0.6146 2 1>, Maximum=<rgba 0.0939 2.7726 3 1>)
     5=&RandomRangeLinearColor(Minimum=<rgba 0.5885 1 0.9678 1>, Maximum=<rgba 2 0.7001 1.8689 1>)
     Index=&MakeIntFromFloat(Float=$User.VariantPicker, Conversion Mode=::NewEnumerator1)
     Num=::NewEnumerator4
    )
    Scale Alpha=&FloatFromCurve(
     FloatCurve=DataInterface() {propertyValues {Curve {keys [
         {interpMode=RCIM_Cubic, tangentMode=RCTM_User, time=-0.0063, value=0.0184}
         {arriveTangent=-0.8483, interpMode=RCIM_Cubic, leaveTangent=-0.8483, tangentMode=RCTM_User, time=0.1718, value=1.7147}
         {arriveTangent=-5.0978, interpMode=RCIM_Cubic, leaveTangent=-5.0978, tangentMode=RCTM_User, time=0.2818, value=1.2143}
         {arriveTangent=-0.2783, interpMode=RCIM_Cubic, leaveTangent=-0.2783, tangentMode=RCTM_User, time=0.4, value=0.7}
         {arriveTangent=0.3112, interpMode=RCIM_Cubic, leaveTangent=0.3112, tangentMode=RCTM_User, time=0.5407, value=1.3419}
         {arriveTangent=-3.7119, interpMode=RCIM_Cubic, leaveTangent=-3.7119, tangentMode=RCTM_User, time=0.7242, value=0.3749}
         {interpMode=RCIM_Cubic, tangentMode=RCTM_User, time=1.0011, value=0.002}
        ]}}}
     Scale Curve=0.8
    )
    Scale Color=&VectorFromFloat(Value=50)
   )
   &ScaleSpriteSize(Scale Factor=&Vector2DFromFloat(Value=&FloatFromCurve(
      FloatCurve=DataInterface() {propertyValues {Curve {keys [
          {arriveTangent=12.7622, interpMode=RCIM_Cubic, leaveTangent=12.7622, tangentMode=RCTM_User, time=0, value=0}
          {arriveTangent=-5.2645, interpMode=RCIM_Cubic, leaveTangent=-5.2645, tangentMode=RCTM_User, time=0.25, value=1.0291}
          {arriveTangent=4.8465, interpMode=RCIM_Cubic, leaveTangent=4.8465, tangentMode=RCTM_User, time=0.6547, value=0.4482}
          {tangentMode=RCTM_User, time=1, value=0}
         ]}}}
      Scale Curve=&Multiply_Float(
       A=&MaskFloatBySpawnGroup(Masked Float=1, Passthrough Float=1.5)
       B=&FN_FrontEndFloatModifier(Gameplay=1, Front End=$System.FrontendScale)
      )
     )))
  ]
  rendererClasses [&NiagaraSpriteRendererProperties]
  renderers [{rendererClass=&NiagaraSpriteRendererProperties}]
 }
}
summary {
 systemName=NS_Pickaxe_SaxoPop_Impact
 emitters [
  {emitterName=NE_Spark, rendererClasses [&NiagaraSpriteRendererProperties]}
  {emitterName=CandyBits, rendererClasses [&NiagaraMeshRendererProperties, &NiagaraSpriteRendererProperties]}
  {emitterName=DarkFlare, rendererClasses [&NiagaraSpriteRendererProperties]}
  {emitterName=Flare_Complex, rendererClasses [&NiagaraSpriteRendererProperties]}
  {
   emitterName=Flash_Elements
   rendererClasses [&NiagaraSpriteRendererProperties, &NiagaraLightRendererProperties]
  }
  {emitterName=Sparkles, rendererClasses [&NiagaraSpriteRendererProperties]}
  {emitterName=SnowFlakes, rendererClasses [&NiagaraSpriteRendererProperties]}
  {emitterName=GalaxySmoke, rendererClasses [&NiagaraMeshRendererProperties, &NiagaraSpriteRendererProperties]}
 ]
 userVariables [
  User.bIsFrontEnd {type=NiagaraBool}
  User.bIsFrontEndPreview {type=NiagaraBool}
  User.FrontEndScale {type=NiagaraFloat, defaultValue=0, description=""}
  User.SizeNormalized {type=NiagaraFloat, defaultValue=0.5, description=""}
  User.VariantPicker {type=NiagaraFloat, defaultValue=0, description=""}
 ]
}
