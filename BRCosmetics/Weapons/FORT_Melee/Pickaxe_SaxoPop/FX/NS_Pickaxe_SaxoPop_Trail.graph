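# Niagara swing trail for the SaxoPop pickaxe cosmetic, with six color variants chosen by User.VariantPicker.
# Trail, Trail_Dark and Trail_Edge draw AnimTrail ribbons between skeletal mesh bones or sockets, with fading alpha.
# Glitter spawns by distance moved, while Sparkles and SnowFlakes spawn at steady rates, all scattered around the mesh.
# VariantPicker also adjusts edge bone order and width, and leaves snowflakes transparent outside variants 3 and 4.
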NiagaraSystem /BRCosmetics/Weapons/FORT_Melee/Pickaxe_SaxoPop/FX/NS_Pickaxe_SaxoPop_Trail
dependencies {
 usedDataInterfaces [&NiagaraDataInterfaceCurve, &NiagaraDataInterfaceColorCurve, &NiagaraDataInterfaceArrayFloat]
 usedDynamicInputs [&SetBoolByIntComparison]
 usedModules [
  &self:SystemSpawnScript.SystemScriptSource.SystemScriptGraph.NiagaraNodeAssignment_0.SetVariables_#1
  &self:SystemSpawnScript.SystemScriptSource.SystemScriptGraph.NiagaraNodeAssignment_4.SetVariables_#2
  &self:SystemSpawnScript.SystemScriptSource.SystemScriptGraph.NiagaraNodeAssignment_5.SetVariables_#3
  &self:SystemSpawnScript.SystemScriptSource.SystemScriptGraph.NiagaraNodeAssignment_6.SetVariables_#4
  &SystemState
  &self:SystemSpawnScript.SystemScriptSource.SystemScriptGraph.NiagaraNodeAssignment_1.SetVariables_#5
  &self:SystemSpawnScript.SystemScriptSource.SystemScriptGraph.NiagaraNodeAssignment_2.SetVariables_#6
  &self:SystemSpawnScript.SystemScriptSource.SystemScriptGraph.NiagaraNodeAssignment_3.SetVariables_#7
 ]
}
emitters {
 Glitter {
  emitterUpdateScript [
   &EmitterState()
   &SpawnPerUnit(Module.Use Max Movement Threshold=true, Module.Use Movement Tolerance=false, Spawn Spacing=50, Use Spawn Probability=false)
  ]
  particleSpawnScript [
   &InitializeParticle(
    Color=&MergeHelper_ColorSwitch(
     0=&RandomRangeLinearColor(Minimum=<rgba 4 2.5287 0.4 1>, Maximum=<rgba 5 0.9696 0.5 1>)
     1=&RandomRangeLinearColor(Minimum=<rgba 2 0.8646 1.6347 0.9957>, Maximum=<rgba 5 1.5104 2.967 1>)
     2=&RandomRangeLinearColor(Minimum=<rgba 1 0.2995 0.0055 1>, Maximum=<rgba 2 1.1 0.2 1>)
     3=&RandomRangeLinearColor(Minimum=<rgba 0.9323 0.977 1 0.9957>, Maximum=<rgba 0.2 1.0139 2 1>)
     4=&RandomRangeLinearColor(Minimum=<rgba 2 0.8646 1.6347 0.9957>, Maximum=<rgba 5 1.5104 2.967 1>)
     5=&RandomRangeLinearColor(Minimum=<rgba 1.952 1.3112 2 0.9957>, Maximum=<rgba 5 1.5104 2.967 1>)
     Index=&MakeIntFromFloat(Float=$User.VariantPicker, Conversion Mode=::NewEnumerator1)
     Num=::NewEnumerator4
    )
    Lifetime Max=0.4
    Lifetime Min=0.3
    Lifetime Mode=::NewEnumerator1
    Sprite Size Mode=::NewEnumerator1
    Uniform Sprite Size Max=1.5
    Uniform Sprite Size Min=0.8
    UsePositionOffset=false
    Write Material Random=true
   )
   &SkeletalMeshLocation(
    Bone / Socket Index=1
    MaskSpawnGroup=false
    Mesh Sampling Type=::NewEnumerator5
    Skeletal Mesh=$System.Skeletal Mesh
    Socket Sampling Mode=::NewEnumerator5
   )
   &InheritVelocity(
    Inherited Velocity Speed Limit:NiagaraFloat=&RandomRangeFloat(Minimum=40, Maximum=80)
    Speed Limit=true
    UseSpeedThreshold=false
   )
   &ApplyInitialForces()
   &ShapeLocation(MaskSpawnGroup=false, Override Seed=false, Sphere Radius=10, UseOwnerScale=true)
  ]
  particleUpdateScript [
   &ParticleState()
   &ScaleSpriteSize(Scale Sprite Size Mode=::NewEnumerator2) {module=ScaleSpriteSize001}
   &ScaleSpriteSize(Uniform Curve Sprite Scale=DataInterface() {propertyValues {Curve {keys [
        {tangentMode=RCTM_User, time=0, value=1}
        {arriveTangent=-0.0405, interpMode=RCIM_Cubic, leaveTangent=-0.0405, tangentMode=RCTM_User, time=0.2, value=1}
        {arriveTangent=-2.1547, interpMode=RCIM_Cubic, leaveTangent=-2.1547, tangentMode=RCTM_User, time=1, value=0}
       ]}}})
   &CurlNoiseForce(Noise Frequency=20, Noise Strength=200, Pan Noise=false)
   &Drag(Drag=&RandomRangeFloat(Minimum=6, Maximum=9), Use Linear Drag=true, Use Rotational Drag=true)
   &SolveForcesAndVelocity(Clamp Velocity=false, Limit Acceleration=false)
   /SetVariables(
    Particles.Color2=&MergeHelper_ColorSwitch(
     0=&RandomRangeLinearColor(Minimum=<rgba 0.0005 0.0008 0.001 1>, Maximum=<rgba 0.1 0.0383 0 1>, Evaluation Type=::NewEnumerator1)
     1=&RandomRangeLinearColor(Minimum=<rgba 0.0005 0.0008 0.001 1>, Maximum=<rgba 0.1 0 0.0574 1>, Evaluation Type=::NewEnumerator1)
     2=&RandomRangeLinearColor(Minimum=<rgba 0.0005 0.0008 0.001 1>, Maximum=<rgba 0.1 0.0383 0 1>, Evaluation Type=::NewEnumerator1)
     3=&RandomRangeLinearColor(Minimum=<rgba 0.0005 0.0008 0.001 1>, Maximum=<rgba 0 0.0383 0.1 1>, Evaluation Type=::NewEnumerator1)
     4=&RandomRangeLinearColor(Minimum=<rgba 0.0005 0.0008 0.001 1>, Maximum=<rgba 0.07 0.0256 0.07 1>, Evaluation Type=::NewEnumerator1)
     5=&RandomRangeLinearColor(Minimum=<rgba 0.0005 0.0008 0.001 1>, Maximum=<rgba 0.07 0.0256 0.07 1>, Evaluation Type=::NewEnumerator1)
     Index=&MakeIntFromFloat(Float=$User.VariantPicker, Conversion Mode=::NewEnumerator1)
     Num=::NewEnumerator4
    )
    Particles.SpriteSize2=&Vector2DFromFloat(Value=10)
   ) {isSetParameters=true}
   &ScaleColor(
    Scale RGB=&VectorFromFloat(Value=&Multiply_Float(
      A=&RandomRangeFloat(Minimum=-10, Maximum=10, Evaluation Type=::NewEnumerator1)
      B=&MakeCustomFloatFromBool(Bool=$User.bIsFrontEndPreview, True Float=1, False Float=1)
     ))
    ScaleA=true
    ScaleRGB=true
   ) {enabled=false}
  ]
  rendererClasses [&NiagaraSpriteRendererProperties]
  renderers [
   {rendererClass=&NiagaraSpriteRendererProperties}
   {rendererClass=&NiagaraSpriteRendererProperties, rendererIndex=1}
  ]
 }
 SnowFlakes {
  emitterUpdateScript [
   &EmitterState(
    Enable Distance Culling=true
    Enable Visibility Culling=true
    Life Cycle Mode=::NewEnumerator1
    Loop Behavior=::NewEnumerator1
    Loop Duration=2
    MaxDistance=2500
    Module.bApplySpawnCountScaleByDistance=true
    Module.bUseMaxDistance=true
    Module.bUseMinDistance=false
    Scalability Mode=::NewEnumerator1
    Scale Spawn Count=true
    UseLoopDelay=false
   )
   &SpawnRate(SpawnRate=20, Use Spawn Probability=false)
   &SpawnRate(SpawnGroup=1, SpawnRate=5, Use Spawn Probability=false) {module=SpawnRate001}
  ]
  particleSpawnScript [
   &InitializeParticle(
    Color=&MergeHelper_ColorSwitch(
     0=<rgba 0 0 0 0>
     1=<rgba 0 0 0 0>
     2=<rgba 0 0 0 0>
     3=&RandomRangeLinearColor(Minimum=<rgba 0.9369 0.9712 1 1>, Maximum=<rgba 0.2865 0.6861 1 1>)
     4=&ColorFromCurve(
      ColorCurve=DataInterface() {propertyValues {
        AlphaCurve {keys [<at 0 1 RCIM_Cubic>]}
        BlueCurve {keys [<at 0.4274 1 RCIM_Cubic>,<at 0.4699 1 RCIM_Cubic>]}
        GreenCurve {keys [<at 0.4274 0.7649 RCIM_Cubic>,<at 0.4699 0.3664 RCIM_Cubic>]}
        RedCurve {keys [<at 0.4274 0.3886 RCIM_Cubic>,<at 0.4699 0.7831 RCIM_Cubic>]}
       }}
      CurveIndex=&RandomRangeFloat(Minimum=0, Maximum=1)
     )
     5=<rgba 0 0 0 0>
     Index=&MakeIntFromFloat(Float=$User.VariantPicker, Conversion Mode=::NewEnumerator1)
     Num=::NewEnumerator4
    )
    Lifetime Max=1
    Lifetime Min=0.15
    Lifetime Mode=::NewEnumerator1
    Sprite Rotation Mode=::NewEnumerator2
    Sprite Size Mode=::NewEnumerator3
    Uniform Sprite Size=&MergeHelper_FloatSwitch(
     0=&RandomRangeFloat(Minimum=1.5, Maximum=5)
     1=&RandomRangeFloat(Minimum=8, Maximum=10)
     2=?
     3=?
     Index=&SpawnGroup()
     Num=::NewEnumerator0
    )
    UsePositionOffset=true
    Write Material Random=true
   )
   &SkeletalMeshLocation(
    Bone / Socket Index=1
    MaskSpawnGroup=false
    Mesh Sampling Type=::NewEnumerator5
    Skeletal Mesh=$System.Skeletal Mesh
    Socket Sampling Mode=::NewEnumerator5
   )
   &ShapeLocation(MaskSpawnGroup=false, Override Seed=false, Sphere Radius=6, Transform Order:ENiagara_TransformOrder=::NewEnumerator1, UseOwnerScale=true) {module=ShapeLocation001}
   &AddVelocity(
    Module.Use Falloff Distance=false
    Override Seed=false
    Velocity Mode=::NewEnumerator1
    Velocity Speed=&RandomRangeFloat(Minimum=0, Maximum=50)
   ) {module=AddVelocity001}
   /SetVariables(Particles.SubImageIndexAlt:NiagaraFloat=&RandomRangeFloat(Minimum=0, Maximum=36)) {isSetParameters=true}
  ]
  particleUpdateScript [
   &ParticleState()
   &SpriteRotationRate(Rotation Rate=&RandomRangeFloat(Minimum=-360, Maximum=360))
   &Drag(Drag=10, Use Linear Drag=true, Use Rotational Drag=true)
   &CurlNoiseForce(Noise Strength=200, Pan Noise=false)
   &GravityForce(Gravity=<0 0 -200>)
   &SolveForcesAndVelocity(Clamp Velocity=false, Limit Acceleration=false)
   &ScaleSpriteSize(Uniform Curve Sprite Scale=DataInterface() {propertyValues {Curve {
       defaultValue=0
       keys [{interpMode=RCIM_Cubic, tangentMode=RCTM_User, time=0, value=1}, {interpMode=RCIM_Cubic, tangentMode=RCTM_User, time=1, value=0}]
      }}})
   &ScaleColor(
    Scale Alpha=&FNTimeOfDayScalar(Night Scale=0.5, Morning Scale=0.75, Evening Scale=0.75)
    ScaleA=true
    ScaleRGB=true
   )
  ]
  rendererClasses [&NiagaraSpriteRendererProperties]
  renderers [
   {rendererClass=&NiagaraSpriteRendererProperties}
   {rendererClass=&NiagaraSpriteRendererProperties, rendererIndex=1}
  ]
 }
 Sparkles {
  emitterUpdateScript [
   &EmitterState(
    Enable Distance Culling=true
    Enable Visibility Culling=true
    Life Cycle Mode=::NewEnumerator1
    Loop Behavior=::NewEnumerator1
    Loop Duration=2
    MaxDistance=2500
    Module.bApplySpawnCountScaleByDistance=true
    Module.bUseMaxDistance=true
    Module.bUseMinDistance=false
    Scalability Mode=::NewEnumerator1
    Scale Spawn Count=true
    UseLoopDelay=false
   )
   &SpawnRate(SpawnRate=75, Use Spawn Probability=false)
   &SpawnRate(SpawnGroup=1, SpawnRate=10, Use Spawn Probability=false) {module=SpawnRate001}
  ]
  particleSpawnScript [
   &InitializeParticle(
    Color=&MergeHelper_ColorSwitch(
     0=&RandomRangeLinearColor(Minimum=<rgba 1 0.4683 0.3281 0.9957>, Maximum=<rgba 2 1.1 0.2 1>)
     1=&RandomRangeLinearColor(Minimum=<rgba 1 0.4323 0.8174 0.9957>, Maximum=<rgba 2 0.6042 1.1868 1>)
     2=&RandomRangeLinearColor(Minimum=<rgba 1 0.2995 0.0055 1>, Maximum=<rgba 2 1.1 0.2 1>)
     3=&RandomRangeLinearColor(Minimum=<rgba 0.9323 0.977 1 0.9957>, Maximum=<rgba 0.2 1.0139 2 1>)
     4=&RandomRangeLinearColor(Minimum=<rgba 0.5885 1 0.9678 1>, Maximum=<rgba 0.0626 1.8484 2 1>)
     5=&RandomRangeLinearColor(Minimum=<rgba 0.5885 1 0.9678 1>, Maximum=<rgba 1.9066 0.6672 2 1>)
     Index=&MakeIntFromFloat(Float=$User.VariantPicker, Conversion Mode=::NewEnumerator1)
     Num=::NewEnumerator4
    )
    Lifetime Max=0.33
    Lifetime Min=0.15
    Lifetime Mode=::NewEnumerator1
    Sprite Rotation Mode=::NewEnumerator2
    Sprite Size Mode=::NewEnumerator3
    Uniform Sprite Size=&MergeHelper_FloatSwitch(
     0=&RandomRangeFloat(Minimum=1.5, Maximum=8)
     1=&RandomRangeFloat(Minimum=15, Maximum=20)
     2=?
     3=?
     Index=&SpawnGroup()
     Num=::NewEnumerator0
    )
    UsePositionOffset=true
    Write Material Random=true
   )
   &SkeletalMeshLocation(
    Bone / Socket Index=1
    MaskSpawnGroup=false
    Mesh Sampling Type=::NewEnumerator5
    Skeletal Mesh=$System.Skeletal Mesh
    Socket Sampling Mode=::NewEnumerator5
   )
   &ShapeLocation(MaskSpawnGroup=false, Override Seed=false, Sphere Radius=6, Transform Order:ENiagara_TransformOrder=::NewEnumerator1, UseOwnerScale=true) {module=ShapeLocation001}
   &AddVelocity(
    Module.Use Falloff Distance=false
    Override Seed=false
    Velocity Mode=::NewEnumerator1
    Velocity Speed=&RandomRangeFloat(Minimum=0, Maximum=50)
   ) {module=AddVelocity001}
  ]
  particleUpdateScript [
   &ParticleState()
   &SpriteRotationRate(Rotation Rate=&RandomRangeFloat(Minimum=-360, Maximum=360))
   &Drag(Drag=10, Use Linear Drag=true, Use Rotational Drag=true)
   &CurlNoiseForce(Noise Strength=200, Pan Noise=false)
   &SolveForcesAndVelocity(Clamp Velocity=false, Limit Acceleration=false)
   &ScaleSpriteSize(Uniform Curve Sprite Scale=DataInterface() {propertyValues {Curve {
       defaultValue=0
       keys [{interpMode=RCIM_Cubic, tangentMode=RCTM_User, time=0, value=1}, {interpMode=RCIM_Cubic, tangentMode=RCTM_User, time=1, value=0}]
      }}})
   &ScaleColor(
    Scale Alpha=&FNTimeOfDayScalar(Night Scale=1, Morning Scale=1, Evening Scale=1)
    ScaleA=true
    ScaleRGB=true
   ) {enabled=false}
  ]
  rendererClasses [&NiagaraSpriteRendererProperties]
  renderers [{rendererClass=&NiagaraSpriteRendererProperties}]
 }
 Trail {
  emitterUpdateScript [
   &EmitterState()
   &FN_SpawnAnimTrail()
  ]
  particleSpawnScript [
   &FN_InitializeAnimTrailBoneIndecies()
   &FN_InitializeAnimTrail(Lifetime=0.33, SkeletalMesh=$System.Skeletal Mesh)
   &CameraOffset(Camera Offset Amount=4)
  ]
  particleUpdateScript [
   &ParticleState()
   &FN_UpdateAnimTrail(SkeletalMesh=$System.Skeletal Mesh)
   &ScaleColor(
    Scale Alpha=&FloatFromCurve(
     FloatCurve=DataInterface() {propertyValues {Curve {
        defaultValue=1
        keys [{interpMode=RCIM_Cubic, tangentMode=RCTM_User, time=0, value=1}, {interpMode=RCIM_Cubic, tangentMode=RCTM_User, time=1, value=0}]
       }}}
     Scale Curve=&FloatFromCurve(
      FloatCurve=DataInterface() {propertyValues {Curve {
         defaultValue=1
         keys [
          {interpMode=RCIM_Cubic, tangentMode=RCTM_User, time=0, value=1}
          {arriveTangent=-22.1023, interpMode=RCIM_Cubic, leaveTangent=-22.1023, tangentMode=RCTM_User, time=0.1, value=0}
         ]
        }}}
      Scale Curve=0.01
     )
    )
    Scale RGB=&MakeVectorFromLinearColorRGB(Color=&MergeHelper_ColorSwitch(
      0=&ColorFromCurve(ColorCurve=DataInterface() {propertyValues {
         AlphaCurve {keys [<at 0 1 RCIM_Cubic>,<at 1 0 RCIM_Cubic>]}
         BlueCurve {keys [<at 0 1.9579 RCIM_Cubic>,<at 1 0.1135 RCIM_Cubic>]}
         GreenCurve {keys [<at 0 1.5249 RCIM_Cubic>,<at 1 0.043 RCIM_Cubic>]}
         RedCurve {keys [<at 0 5 RCIM_Cubic>,<at 1 1 RCIM_Cubic>]}
        }})
      1=&ColorFromCurve(ColorCurve=DataInterface() {propertyValues {
         AlphaCurve {keys [<at 0 1 RCIM_Cubic>,<at 1 0 RCIM_Cubic>]}
         BlueCurve {keys [<at 0 1.71 RCIM_Cubic>,<at 0.9969 0.7258>]}
         GreenCurve {keys [<at 0 0.9586 RCIM_Cubic>,<at 0.9969 0.2575>]}
         RedCurve {keys [<at 0 2 RCIM_Cubic>,<at 0.9969 1>]}
        }})
      2=&ColorFromCurve(ColorCurve=DataInterface() {propertyValues {
         AlphaCurve {keys [<at 0 1 RCIM_Cubic>,<at 1 0 RCIM_Cubic>]}
         BlueCurve {keys [<at 0 0.5104 RCIM_Cubic>,<at 0.9969 0>]}
         GreenCurve {keys [<at 0 1.653 RCIM_Cubic>,<at 0.9969 0.2609>]}
         RedCurve {keys [<at 0 2 RCIM_Cubic>,<at 0.9969 1>]}
        }})
      3=&ColorFromCurve(ColorCurve=DataInterface() {propertyValues {
         AlphaCurve {keys [<at 0 1 RCIM_Cubic>,<at 1 0 RCIM_Cubic>]}
         BlueCurve {keys [<at 0.4884 2 RCIM_Cubic>,<at 0.9969 1>]}
         GreenCurve {keys [<at 0.4884 1.8966 RCIM_Cubic>,<at 0.9969 0.1913>]}
         RedCurve {keys [<at 0.4884 0.6783 RCIM_Cubic>,<at 0.9969 0>]}
        }})
      4=&ColorFromCurve(ColorCurve=DataInterface() {propertyValues {
         AlphaCurve {keys [<at 0 1 RCIM_Cubic>,<at 1 0 RCIM_Cubic>]}
         BlueCurve {keys [<at 0.4884 3 RCIM_Cubic>,<at 0.9969 0.5029>]}
         GreenCurve {keys [<at 0.4884 2.9179 RCIM_Cubic>,<at 0.9969 0.4125>]}
         RedCurve {keys [<at 0.4884 0.25 RCIM_Cubic>,<at 0.9969 0.0704>]}
        }})
      5=&ColorFromCurve(ColorCurve=DataInterface() {propertyValues {
         AlphaCurve {keys [<at 0 1 RCIM_Cubic>,<at 1 0 RCIM_Cubic>]}
         BlueCurve {keys [<at 0.4884 7.8282 RCIM_Cubic>,<at 0.9969 0.2557>]}
         GreenCurve {keys [<at 0.4884 1.375 RCIM_Cubic>,<at 0.9969 0>]}
         RedCurve {keys [<at 0.4884 8 RCIM_Cubic>,<at 0.9969 0.2625>]}
        }})
      Index=&MakeIntFromFloat(Float=$User.VariantPicker, Conversion Mode=::NewEnumerator1)
      Num=::NewEnumerator4
     ))
    ScaleA=true
    ScaleRGB=true
   )
  ]
  rendererClasses [&NiagaraRibbonRendererProperties]
  renderers [{rendererClass=&NiagaraRibbonRendererProperties}]
 }
 Trail_Dark {
  emitterUpdateScript [
   &EmitterState()
   &FN_SpawnAnimTrail_v2()
  ]
  particleSpawnScript [
   /Game/Effects/Niagara/Modules/AnimTrail/FN_InitializeAnimTrail/FN_InitializeAnimTrailBoneIndecies(Second Bone or Socket Index=1)
   &FN_InitializeAnimTrail(
    AnimTrailWidth=2
    Color=&MergeHelper_ColorSwitch(
     0=<rgba 0.1875 0 0.025 1.2>
     1=<rgba 0.1739 0 0.0907 1>
     2=<rgba 0.174 0.0092 0.0001 1.2>
     3=<rgba 0 0.011 0.0677 2>
     4=<rgba 0.1042 0.045 0.0894 3>
     5=<rgba 0.0045 0 0.0307 2.5>
     Index=&MakeIntFromFloat(Float=$User.VariantPicker)
     Num=::NewEnumerator4
    )
    Lifetime=0.33
    SkeletalMesh=$System.Skeletal Mesh
   ) {module=FN_InitializeAnimTrail001}
   /SetVariables(Particles.Position_Tip:NiagaraPosition=&AddVectorToPosition(Position=$Particles.Position, Vector=&TransformVector(Vector=<0 0 0>, Destination Space=::Simulation))) {isSetParameters=true}
  ]
  particleUpdateScript [
   &FN_UpdateAnimTrail(SkeletalMesh=$System.Skeletal Mesh)
   &ScaleRibbonWidth(Ribbon Width Scale=&FloatFromCurve(FloatCurve=DataInterface() {propertyValues {Curve {keys [
         {interpMode=RCIM_Cubic, tangentMode=RCTM_User, time=0, value=1.1}
         {arriveTangent=-2.3701, interpMode=RCIM_Cubic, leaveTangent=-2.3701, tangentMode=RCTM_User, time=1, value=0}
        ]}}}))
   &ScaleColor(
    Scale Alpha=&FloatFromCurve(
     FloatCurve=DataInterface() {propertyValues {Curve {
        defaultValue=1
        keys [
         {interpMode=RCIM_Cubic, tangentMode=RCTM_User, time=0, value=1}
         {arriveTangent=-21.2207, interpMode=RCIM_Cubic, leaveTangent=-21.2207, tangentMode=RCTM_User, time=0.15, value=0}
        ]
       }}}
     Scale Curve=0.25
    )
    ScaleA=true
    ScaleRGB=true
   )
   &SpriteFacingAndAlignment(Sprite Facing=$Particles.RibbonFacing, WriteAlignment=false, WriteFacing=true)
   &DynamicMaterialParameters(
    Index 0 Param 1=&FloatFromCurve(FloatCurve=DataInterface() {propertyValues {Curve {keys [<at 0 1 RCIM_Cubic>, {arriveTangent=-1, interpMode=RCIM_Cubic, leaveTangent=-1, time=0.5, value=0.8}, <at 1 0>]}}})
    Module.Param0WriteEnabled=true
    Module.Param1WriteEnabled=true
    Module.Param2WriteEnabled=true
    Module.Param3WriteEnabled=true
   )
  ]
  rendererClasses [&NiagaraRibbonRendererProperties]
  renderers [{rendererClass=&NiagaraRibbonRendererProperties}]
 }
 Trail_Edge {
  emitterUpdateScript [
   &EmitterState()
   &FN_SpawnAnimTrail()
   /SetVariables(Emitter.rand:NiagaraFloat=&RandomRangeFloat(Minimum=1, Maximum=1, Evaluation Type=::NewEnumerator1)) {isSetParameters=true}
  ]
  particleSpawnScript [
   /Game/Effects/Niagara/Modules/AnimTrail/FN_InitializeAnimTrailBoneIndecies_v2/FN_InitializeAnimTrailBoneIndecies(
    First Bone or Socket Index:NiagaraInt32=&MergeHelper_IntSwitch(
     0=0
     1=0
     2=0
     3=0
     4=0
     5=1
     Index=&MakeIntFromFloat(Float=$User.VariantPicker, Conversion Mode=::NewEnumerator1)
     Num=::NewEnumerator4
    )
    Second Bone or Socket Index:NiagaraInt32=&MergeHelper_IntSwitch(0=1, 1=1, 2=1, 3=1, 4=1, 5=0, Index=&MakeIntFromFloat(Float=$User.VariantPicker), Num=::NewEnumerator4)
   )
   &FN_InitializeAnimTrail(Lifetime=0.33, SkeletalMesh=$System.Skeletal Mesh)
  ]
  particleUpdateScript [
   &FN_UpdateAnimTrail(SkeletalMesh=$System.Skeletal Mesh)
   &ParticleState()
   &ScaleColor(
    Scale Alpha=&Multiply_Float(
     A=0.8
     B=&FloatFromCurve(
      FloatCurve=DataInterface() {propertyValues {Curve {keys [
          {arriveTangent=-0.0004, interpMode=RCIM_Cubic, leaveTangent=-0.0004, tangentMode=RCTM_User, time=0, value=1}
          {arriveTangent=-0.3873, interpMode=RCIM_Cubic, leaveTangent=-0.3873, tangentMode=RCTM_User, time=0.15, value=0}
         ]}}}
      Scale Curve=$Emitter.rand
     )
    )
    Scale RGB=&MakeVectorFromLinearColorRGB(Color=&MergeHelper_ColorSwitch(
      0=&ColorFromCurve(ColorCurve=DataInterface() {propertyValues {
         AlphaCurve {keys [<at 0 1 RCIM_Cubic>,<at 1 0 RCIM_Cubic>]}
         BlueCurve {keys [
           <at 0.0384 0.4164>
           {arriveTangent=-0.9233, interpMode=RCIM_Cubic, leaveTangent=-0.9233, time=0.133, value=0.1765}
           <at 0.393 0.089>
           <at 0.7988 0.025 RCIM_Cubic>
          ]}
         GreenCurve {keys [
           <at 0.0384 0.3281>
           {arriveTangent=-0.8166, interpMode=RCIM_Cubic, leaveTangent=-0.8166, time=0.133, value=0.1001}
           <at 0.393 0.0385>
           <at 0.7988 0 RCIM_Cubic>
          ]}
         RedCurve {keys [
           <at 0.0384 1>
           {arriveTangent=-1.3219, interpMode=RCIM_Cubic, leaveTangent=-1.3219, time=0.133, value=0.8}
           <at 0.393 0.5312>
           <at 0.7988 0.1875 RCIM_Cubic>
          ]}
        }})
      1=&ColorFromCurve(ColorCurve=DataInterface() {propertyValues {
         AlphaCurve {keys [<at 0 1 RCIM_Cubic>,<at 1 0 RCIM_Cubic>]}
         BlueCurve {keys [<at 0.0384 0.7913>,<at 0.1479 0.2721>,<at 0.4234 0.1488>,<at 0.8177 0.0907>]}
         GreenCurve {keys [<at 0.0384 0.4072>,<at 0.1479 0.1474>,<at 0.4234 0.0472>,<at 0.8177 0>]}
         RedCurve {keys [<at 0.0384 0.9473>,<at 0.1479 0.7448>,<at 0.4234 0.3565>,<at 0.8177 0.1739>]}
        }})
      2=&ColorFromCurve(ColorCurve=DataInterface() {propertyValues {
         AlphaCurve {keys [<at 0 1 RCIM_Cubic>,<at 1 0 RCIM_Cubic>]}
         BlueCurve {keys [<at 0.0384 0.251>,<at 0.1539 0.0055>,<at 0.8439 0>]}
         GreenCurve {keys [<at 0.0384 0.7695>,<at 0.1539 0.2995>,<at 0.8439 0.0563>]}
         RedCurve {keys [<at 0.0384 0.9>,<at 0.1539 1>,<at 0.8439 0.5391>]}
        }})
      3=&ColorFromCurve(ColorCurve=DataInterface() {propertyValues {
         AlphaCurve {keys [<at 0 1 RCIM_Cubic>,<at 1 0 RCIM_Cubic>]}
         BlueCurve {keys [<at 0.0384 0.9>,<at 0.1539 1>,<at 0.8654 0.5391>]}
         GreenCurve {keys [<at 0.0384 0.7815>,<at 0.1539 0.3514>,<at 0.8654 0.1594>]}
         RedCurve {keys [<at 0.0384 0.251>,<at 0.1539 0.0055>,<at 0.8654 0>]}
        }})
      4=&ColorFromCurve(ColorCurve=DataInterface() {propertyValues {
         AlphaCurve {keys [<at 0 1 RCIM_Cubic>,<at 1 0 RCIM_Cubic>]}
         BlueCurve {keys [<at 0 0.2>,<at 0.8807 0.1>]}
         GreenCurve {keys [<at 0 0.2>,<at 0.8807 0.1>]}
         RedCurve {keys [<at 0 0.2>,<at 0.8807 0.1>]}
        }})
      5=&ColorFromCurve(ColorCurve=DataInterface() {propertyValues {
         AlphaCurve {keys [<at 0 1 RCIM_Cubic>,<at 1 0 RCIM_Cubic>]}
         BlueCurve {keys [<at 0 0.9216>,<at 0.9145 0.2552>]}
         GreenCurve {keys [<at 0 0.6172>,<at 0.9145 0.2208>]}
         RedCurve {keys [<at 0 0.7682>,<at 0.9145 0.0505>]}
        }})
      Index=&MakeIntFromFloat(Float=$User.VariantPicker, Conversion Mode=::NewEnumerator1)
      Num=::NewEnumerator4
     ))
    ScaleA=true
    ScaleRGB=true
   )
   /SetVariables(Particles.SpecificRibbonWidth:NiagaraFloat=&Multiply_Float(
     A=$Particles.RibbonWidth
     B=&SelectFloatFromArray(
      Direct Array Index=&MakeIntFromFloat(Float=$User.VariantPicker, Conversion Mode=::NewEnumerator1)
      Float Selection Array=DataInterface() {propertyValues {FloatData [0,0,0,0,1.2,2]}}
     )
    )) {isSetParameters=true}
  ]
  rendererClasses [&NiagaraRibbonRendererProperties]
  renderers [
   {rendererClass=&NiagaraRibbonRendererProperties}
   {rendererClass=&NiagaraRibbonRendererProperties, rendererIndex=1}
   {rendererClass=&NiagaraRibbonRendererProperties, rendererIndex=2}
  ]
 }
}
summary {
 systemName=NS_Pickaxe_SaxoPop_Trail
 emitters [
  {emitterName=Glitter, rendererClasses [&NiagaraSpriteRendererProperties]}
  {emitterName=Trail_Edge, rendererClasses [&NiagaraRibbonRendererProperties]}
  {emitterName=Sparkles, rendererClasses [&NiagaraSpriteRendererProperties]}
  {emitterName=Trail, rendererClasses [&NiagaraRibbonRendererProperties]}
  {emitterName=Trail_Dark, rendererClasses [&NiagaraRibbonRendererProperties]}
  {emitterName=SnowFlakes, rendererClasses [&NiagaraSpriteRendererProperties]}
 ]
 userVariables [
  User.bIsFrontEnd {type=NiagaraBool}
  User.bIsFrontEndPreview {type=NiagaraBool}
  User.bIsOffhand {type=NiagaraBool}
  User.bIsRightHand {type=NiagaraBool}
  User.Color A {type=LinearColor, description="", defaultValue {struct=LinearColor, value=<rgba 1 1 1 1>}}
  User.ColorSwitcher {type=NiagaraFloat, defaultValue=0, description=""}
  User.Debug {type=NiagaraBool}
  User.IsR2 {type=NiagaraBool}
  User.RibbonIDOffset {type=NiagaraInt32}
  User.RibbonWidth {type=NiagaraFloat, defaultValue=10, description=""}
  User.SizeNormalized {type=NiagaraFloat, defaultValue=0.5, description=""}
  User.SkeletalMesh {
   type=NiagaraDataInterfaceSkeletalMesh
   defaultValue {
    struct=NiagaraExt_VariableValue_DataInterface
    value {dataInterface=&self:NiagaraDataInterfaceSkeletalMesh_0}
   }
  }
  User.SwingSectionIndex {type=NiagaraInt32}
  User.VariantPicker {type=NiagaraFloat, defaultValue=0, description=""}
 ]
}
