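# Material function for the SaxoPop pickaxe that fakes a glittery translucent look, overriding color, emissive, roughness and metallic while passing the input normal through.
# Inner stars come from two local-axis virtual planes blended by view angle with time-animated radius, alongside bump-offset dust.
# A refracted skylight sample and two-color gradient build base color.
# Fake subsurface scattering uses Fresnel and sun direction, and emissive adjusts by time of day.
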MaterialFunction /BRCosmetics/Weapons/FORT_Melee/Pickaxe_SaxoPop/Materials/MF_SaxoPop_Pickaxe
graph {nodes [
  n1 = FunctionInput() {inputType=FunctionInput_MaterialAttributes}
  n2 = SetMaterialAttributes!(
   MaterialAttributes
   BaseColor=NamedRerouteUsage()
   EmissiveColor=NamedRerouteUsage()
   Roughness=NamedRerouteUsage()
   Normal=GetMaterialAttributes(None=NamedRerouteDeclaration(None=@n1) {name=InMaterial, nodeColor=<rgba 1 0.5859 0 1>}) {attributeGetTypes [#1]}.Normal
   Metallic=ScalarParameter() {defaultValue=0.3, group=SaxoPop, parameterName=Metalness}
  ) {attributeSetTypes [
    #2
    #3
    #4
    #1
    #5
   ]}
  FunctionOutput(None=@n2)
  n3 = NamedRerouteUsage()
  Multiply(A=@n3, B=OneMinus(None=NamedRerouteUsage()))
  n4 = Multiply(A=SkyAtmosphereLightDirection()) {constB=-1}
  n5 = DotProduct(A=@n4, B=PixelNormalWS())
  NamedRerouteDeclaration(None=@n5) {name=NdotL, nodeColor=<rgba 1 0 0.5156 1>}
  n6 = ConstantBiasScale(None=NamedRerouteUsage())
  n7 = NamedRerouteUsage()
  n8 = NamedRerouteUsage()
  n9 = LinearInterpolate(A=ScalarParameter() {defaultValue=0.3, parameterName=A}, Alpha=Multiply(A=@n8, B=@n8)) {constA=0.25}
  n10 = Multiply(
   A=Saturate(None=Multiply(A=@n7, B=@n9))
   B=Power(Base=NamedRerouteUsage(), Exp=ScalarParameter() {defaultValue=1.5, group=SaxoPop, parameterName=StarsFadeExp}) {constExponent=1.5}
  )
  NamedRerouteDeclaration(None=Multiply(A=Power(Base=@n6) {constExponent=1.5}, B=NamedRerouteDeclaration(None=@n10) {name=Stars, nodeColor=<rgba 0 1 0.3438 1>})) {name=StarsNdotL, nodeColor=<rgba 1 0.1172 0 1>}
  n11 = ComponentMask(None=PreSkinnedNormal()) {b=true}
  n12 = Multiply(A=@n11, B=@n11)
  NamedRerouteDeclaration(None=VertexInterpolator(VS=SmoothStep(Value=@n12) {constMax=0.1}).PS) {name=TipMask, nodeColor=<rgba 0 1 0.4141 1>}
  n13 = Transform(None=Constant3Vector() {constant=<rgba 1 0 0 1>}) {transformSourceType=TRANSFORMSOURCE_Local}
  n14 = NamedRerouteDeclaration(None=@n13) {name=WS_Axis_X, nodeColor=<rgba 0 0.4375 1 1>}
  n15 = Transform(None=Constant3Vector() {constant=<rgba 0 0 1 1>}) {transformSourceType=TRANSFORMSOURCE_Local}
  n16 = Normalize(VectorInput=CrossProduct(A=@n13, B=@n15))
  n17 = Add(
   A=MaterialFunctionCall() {materialFunction=&ObjectPivotPoint}.Object Pivot Location
   B=Transform(None=Constant3Vector() {constant=<rgba 0 0 25 0>}) {transformSourceType=TRANSFORMSOURCE_Local}
  )
  n18 = ScalarParameter() {defaultValue=15, group=SaxoPop, parameterName=StarsTiling}
  n19 = TextureSample(UVs=NamedRerouteDeclaration(None=MaterialFunctionCall(Plane Normal Axis=@n14, Plane X Axis=@n16, Plane Center=@n17, UV Scale=@n18) {materialFunction=&VirtualPlaneCoordinates}.UVs) {name=XPlane, nodeColor=<rgba 0 0.8359 1 1>}) {samplerType=SAMPLERTYPE_Masks, texture=&/T_SaxoPop_Map}
  n20 = Transform(None=Constant3Vector() {constant=<rgba 0 1 0 1>}) {transformSourceType=TRANSFORMSOURCE_Local}
  n21 = Normalize(VectorInput=CrossProduct(A=@n20, B=@n15))
  n22 = TextureSample(UVs=NamedRerouteDeclaration(None=MaterialFunctionCall(Plane Normal Axis=@n20, Plane X Axis=@n21, Plane Center=@n17, UV Scale=@n18) {materialFunction=&VirtualPlaneCoordinates}.UVs) {name=YPlane, nodeColor=<rgba 1 0.8438 0 1>}) {samplerType=SAMPLERTYPE_Masks, texture=&T_SaxoPop_Map}
  n23 = Max(A=@n19.R, B=@n22.R)
  n24 = Constant() {r=1}
  n25 = NamedRerouteUsage()
  n26 = Constant() {r=1}
  n27 = Abs(None=DotProduct(
    A=NamedRerouteUsage()
    B=MaterialFunctionCall() {materialFunction=&CameraDirectionVector}.Result
   ))
  n28 = LinearInterpolate(
   A=Multiply(A=SphereMask(A=@n19.A, B=@n24, Radius=@n25) {attenuationRadius=1}, B=NamedRerouteUsage())
   B=Multiply(A=SphereMask(A=@n22.A, B=@n26, Radius=@n25) {attenuationRadius=1}, B=NamedRerouteUsage())
   Alpha=OneMinus(None=Multiply(A=@n27, B=@n27))
  )
  n29 = OneMinus(None=Power(Base=NamedRerouteUsage()) {constExponent=16})
  NamedRerouteDeclaration(None=Multiply(A=LinearInterpolate(Alpha=@n23) {constA=0.5}, B=Multiply(A=@n28, B=@n29))) {name=OutVPStars, nodeColor=<rgba 0 0.6484 1 1>}
  n30 = Time()
  n31 = Sine(None=Subtract(A=Multiply(A=@n30) {constB=0.6}, B=@n23))
  NamedRerouteDeclaration(None=LinearInterpolate(Alpha=ConstantBiasScale(None=@n31)) {constA=0.4, constB=0.7}) {name=StarRandomRadius, nodeColor=<rgba 1 0 0.1172 1>}
  n32 = TextureCoordinate()
  n33 = Multiply(A=@n32) {constB=1.75}
  n34 = Add(A=@n33, B=ScalarParameter() {defaultValue=0.2, parameterName=FrontStarOffset})
  n35 = Multiply(A=@n32) {constB=0.1}
  n36 = TextureSample(UVs=@n35) {texture=&AtmosphericCloudNoise01}
  n37 = LinearInterpolate(Alpha=@n36.G) {constA=3}
  n38 = Add(A=Multiply(A=@n33) {constB=2}, B=Constant2Vector() {g=0.608, r=0.416}) {constB=0.25}
  n39 = LinearInterpolate(Alpha=@n36.G) {constA=1, constB=3}
  n40 = Max(
   A=TextureSample(UVs=MaterialFunctionCall!(UV Coord=@n34, Height=@n37, ReferencePlane, HeightRatio, Normal) {materialFunction=/Game/Global/MaterialFunctions/MF_BumpOffset_Accurate}.Result) {samplerType=SAMPLERTYPE_Masks, texture=&T_SaxoPop_Map}.B
   B=TextureSample(UVs=MaterialFunctionCall(UV Coord=@n38, Height=@n39) {materialFunction=&MF_BumpOffset_Accurate}.Result) {samplerType=SAMPLERTYPE_Masks, texture=&T_SaxoPop_Map}.B
  )
  n41 = OneMinus(None=Fresnel() {baseReflectFraction=0, exponent=0.5})
  NamedRerouteDeclaration(None=Max(A=@n3, B=Multiply(A=@n40, B=@n41))) {name=Dust, nodeColor=<rgba 0.1016 0 1 1>}
  NamedRerouteDeclaration(None=TextureSample(UVs=BumpOffset(Coordinate=@n35, Height=@n36.G)) {texture=&AtmosphericCloudNoise01}.G) {name=BumpedNoise, nodeColor=<rgba 0.2187 1 0 1>}
  n42 = ScalarParameter() {defaultValue=0.7, group=SaxoPop, parameterName=StarsMaskExtent}
  n43 = NamedRerouteUsage()
  n44 = Subtract(A=ComponentMask(None=@n43) {r=true}) {constB=0.5}
  n45 = Ceil(None=Saturate(None=Abs(None=@n44)))
  n46 = ComponentMask(None=@n43) {r=true}
  n47 = OneMinus(None=Abs(None=Subtract(A=@n46) {constB=0.5}))
  n48 = NamedRerouteDeclaration(None=SmoothStep(Min=@n42, Value=Multiply(A=@n45, B=@n47)) {constMin=0.5}) {name=XPlaneMask, nodeColor=<rgba 0.5234 0 1 1>}
  n49 = NamedRerouteUsage()
  n50 = Subtract(A=ComponentMask(None=@n49) {r=true}) {constB=0.5}
  n51 = Ceil(None=Saturate(None=Abs(None=@n50)))
  n52 = ComponentMask(None=@n49) {r=true}
  n53 = OneMinus(None=Abs(None=Subtract(A=@n52) {constB=0.5}))
  n54 = NamedRerouteDeclaration(None=SmoothStep(Min=@n42, Value=Multiply(A=@n51, B=@n53)) {constMin=0.5}) {name=YPlaneMask, nodeColor=<rgba 0 1 0.2031 1>}
  n55 = Multiply(
   A=Power(Base=Max(A=@n48, B=@n54), Exp=ScalarParameter() {defaultValue=1.5, parameterName=Exp_1}) {constExponent=3}
   B=ScalarParameter() {defaultValue=1.25, parameterName=B}
  ) {constB=2}
  NamedRerouteDeclaration(None=@n55) {name=Vertical Gradient test, nodeColor=<rgba 0 1 0.7188 1>}
  n56 = Add(A=LocalPosition() {localOrigin=InstancePreSkinning}.XYZ, B=Constant3Vector() {constant=<rgba 0 0 -11.2553 0>})
  n57 = ComponentMask(None=NamedRerouteDeclaration(None=VertexInterpolator(VS=@n56).PS) {name=PreSkinnedPosOffseted, nodeColor=<rgba 0 0.25 1 1>}) {b=true}
  NamedRerouteDeclaration(None=Saturate(None=Multiply(A=@n57) {constB=0.0325})) {name=VerticalGradient, nodeColor=<rgba 0 0.0625 1 1>}
  n58 = LinearInterpolate(
   A=ScalarParameter() {defaultValue=0.1, group=SaxoPop, parameterName=MinEmissive}
   B=ScalarParameter() {defaultValue=7, group=SaxoPop, parameterName=FakeSSSIntensity}
   Alpha=NamedRerouteUsage()
  ) {constA=1, constB=3}
  n59 = NamedRerouteUsage()
  n60 = OneMinus(None=Fresnel() {baseReflectFraction=0, exponent=2})
  n61 = Multiply(A=NamedRerouteUsage()) {constB=-1}
  n62 = Multiply(
   A=Power(Base=Saturate(None=@n61), Exp=ScalarParameter() {defaultValue=2, parameterName=UpDotCamEmissiveGainExp}) {constExponent=4}
   B=ScalarParameter() {defaultValue=0.5, parameterName=UpDotCamEmissiveGainInt}
  )
  n63 = Multiply(A=NamedRerouteUsage(), B=Add(A=Add(A=@n58, B=@n59), B=Multiply(A=@n60, B=@n62)) {constB=0})
  n64 = Multiply(A=@n63) {constB=0.75}
  n65 = NamedRerouteUsage()
  NamedRerouteDeclaration(None=Add(
    A=MaterialFunctionCall(Color_Morning=@n63, Color_Day=@n63, Color_Evening=@n64, Color_Night=@n64) {materialFunction=&Color_Time_of_Day}.Result
    B=Multiply(A=@n65) {constB=1.5}
   )) {name=OutEmissiveColor, nodeColor=<rgba 0.3828 1 0 1>}
  NamedRerouteDeclaration(None=LinearInterpolate(
    A=ScalarParameter() {defaultValue=0.3, group=SaxoPop, parameterName=RoughnessMin}
    B=ScalarParameter() {defaultValue=0.4, group=SaxoPop, parameterName=RoughnessMax}
    Alpha=TextureSample() {samplerType=SAMPLERTYPE_Masks, texture=&T_SaxoPop_Map}.G
   )) {name=OutRoughness, nodeColor=<rgba 1 0.4922 0 1>}
  n66 = ScalarParameter() {defaultValue=3, group=SaxoPop, parameterName=FresnelExponent}
  n67 = DotProduct(
   A=Transform(None=Constant3Vector() {constant=<rgba 0 0 1 0>}) {transformSourceType=TRANSFORMSOURCE_Local}
   B=MaterialFunctionCall() {materialFunction=&CameraDirectionVector}.Result
  )
  n68 = NamedRerouteDeclaration(None=Abs(None=NamedRerouteDeclaration(None=@n67) {name=UpDotCam, nodeColor=<rgba 0 0.2031 1 1>})) {name=UpDotCamAbs, nodeColor=<rgba 0 1 0.7422 1>}
  n69 = NamedRerouteDeclaration(None=Multiply(A=Fresnel(ExponentIn=@n66) {baseReflectFraction=0, exponent=2}, B=OneMinus(None=@n68))) {name=Fresnel, nodeColor=<rgba 0.3828 1 0 1>}
  n70 = OneMinus(None=Fresnel() {baseReflectFraction=0, exponent=0.5})
  n71 = CameraVectorWS()
  n72 = Multiply(A=Power(Base=DotProduct(A=@n71, B=@n4), Exp=ScalarParameter() {defaultValue=3, group=SaxoPop, parameterName=Exp}), B=@n5)
  NamedRerouteDeclaration(None=Add(A=Multiply(A=@n69) {constB=0.5}, B=Multiply(A=@n70, B=@n72))) {name=OutFakeSSS, nodeColor=<rgba 0 1 0.7891 1>}
  n73 = Normalize(VectorInput=Multiply(A=CameraVectorWS(), B=Constant() {r=-1}) {constB=-1})
  n74 = PixelNormalWS()
  NamedRerouteDeclaration(None=SkyLightEnvMapSample(
    Direction=MaterialFunctionCall!(Ray Direction=@n73, Surface Normal=@n74, Refractive Index Origin, Refractive Index Target) {materialFunction=/Engine/Functions/Engine_MaterialFunctions01/Vectors/Refract}.Vector
    Roughness=Constant()
   )) {name=EnvMap, nodeColor=<rgba 0.5 0 1 1>}
  n75 = NamedRerouteUsage()
  n76 = ScalarParameter() {defaultValue=0.5, group=SaxoPop, parameterName=EnvDesaturation}
  n77 = NamedRerouteUsage()
  n78 = OneMinus(None=NamedRerouteUsage())
  n79 = Power(Base=NamedRerouteUsage()) {constExponent=0.25}
  n80 = Power(Base=NamedRerouteUsage())
  n81 = Multiply(A=Add(A=Multiply(A=@n77, B=@n78), B=Multiply(A=@n78) {constB=0.5}), B=LinearInterpolate(Alpha=Max(A=@n79, B=@n80)) {constA=0.6})
  n82 = NamedRerouteDeclaration(None=LinearInterpolate(
    A=VectorParameter() {group=SaxoPop, parameterName=UpperColor, defaultValue=<rgba 0.6823 0.0475 0.0475 1>}.RGB
    B=VectorParameter() {group=SaxoPop, parameterName=BottomColor, defaultValue=<rgba 1 0.1033 0.0909 1>}.RGB
    Alpha=@n81
   )) {name=Colors, nodeColor=<rgba 1 0 0.2344 1>}
  n83 = LinearInterpolate(
   A=Multiply(A=Desaturation(None=@n75, Fraction=@n76), B=@n82)
   B=@n82
   Alpha=LinearInterpolate(
    A=ScalarParameter() {group=SaxoPop, parameterName=MinFakeTranslucency}
    Alpha=NamedRerouteDeclaration(None=@n81) {name=Gradient, nodeColor=<rgba 0 1 0.0859 1>}
   ) {constA=0.25}
  )
  n84 = ScalarParameter() {defaultValue=0.5, group=SaxoPop, parameterName=BaseColorIntensity}
  NamedRerouteDeclaration(None=Add(A=Multiply(A=@n83, B=@n84) {constB=0.5}, B=NamedRerouteUsage())) {name=OutBaseColor, nodeColor=<rgba 0 1 0.2266 1>}
  Saturate(None=@n55)
 ]}
